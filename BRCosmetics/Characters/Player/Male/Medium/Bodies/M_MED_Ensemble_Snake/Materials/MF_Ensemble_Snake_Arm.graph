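# Material function for the Snake ensemble arm that adds a dithered opacity fade and a vertex push to incoming material attributes, behind the UseEnsembleSnakeArm static switch.
# MaterialCurveA and MaterialCurveB shift bands along UV1's vertical axis, forming the push and opacity gradients.
# Vertex color supplies arm mask (R), sleeve mask (G), signed push direction (B) and vertex mask (A).
# These masks and gradients feed world position offset and temporally dithered opacity mask.
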MaterialFunction /BRCosmetics/Characters/Player/Male/Medium/Bodies/M_MED_Ensemble_Snake/Materials/MF_Ensemble_Snake_Arm
graph {nodes [
  FunctionOutput(None=StaticSwitchParameter(True=NamedRerouteUsage(), False=NamedRerouteUsage()) {group=M_Ensemble_Snake_Arm, parameterName=UseEnsembleSnakeArm})
  n1 = ComponentMask(None=TextureCoordinate() {coordinateIndex=1}) {g=true}
  n2 = Add(A=Multiply(A=@n1) {constB=2}) {constB=0.4}
  n3 = ScalarParameter() {group=M_Ensemble_Snake_Arm, parameterName=MaterialCurveA, sliderMax=1}
  n4 = Add(A=Multiply(A=@n1) {constB=0.5}) {constB=0.25}
  n5 = SphereMask(
   A=Add(A=@n4, B=OneMinus(None=@n3))
   B=Constant() {r=1}
   Radius=ScalarParameter() {defaultValue=0.2, group=M_Ensemble_Snake_Arm, parameterName=PushBand_Width, sliderMax=1}
  ) {hardnessPercent=0}
  Multiply(A=Saturate(None=Power(Base=@n2) {constExponent=4}), B=@n5)
  NamedRerouteDeclaration(None=FunctionInput() {inputName=MA, inputType=FunctionInput_MaterialAttributes}) {name=MA, nodeColor=<rgba 1 0.0234 0 1>}
  n6 = MaterialFunctionCall(Alpha Threshold=Multiply(A=NamedRerouteUsage(), B=NamedRerouteUsage())) {materialFunction=&DitherTemporalAA}
  n7 = NamedRerouteUsage()
  n8 = Multiply(A=NamedRerouteUsage(), B=ScalarParameter() {defaultValue=4, group=M_Ensemble_Snake_Arm, parameterName=PushIntensity, sliderMax=10})
  n9 = Multiply(A=Multiply(A=NamedRerouteUsage(), B=NamedRerouteUsage()), B=NamedRerouteUsage())
  NamedRerouteDeclaration(None=SetMaterialAttributes!(MaterialAttributes=NamedRerouteUsage(), OpacityMask=Add(A=@n6.Result, B=@n7), WorldPositionOffset=Multiply(A=@n8, B=@n9), EmissiveColor) {attributeSetTypes [#1, #2, #3]}) {name=MA_Out, nodeColor=<rgba 1 0.1875 0 1>}
  n10 = Constant()
  n11 = AppendVector(A=@n10, B=@n10)
  n12 = VertexColor()
  n13 = Add(A=Multiply(A=@n12.B) {constB=2}) {constB=-1}
  NamedRerouteDeclaration(None=Transform(None=AppendVector(A=@n11, B=@n13))) {name=VertexPushDirection, nodeColor=<rgba 0 0.1304 1 1>}
  NamedRerouteDeclaration(None=@n12.A) {name=VertexMask, nodeColor=<rgba 0.9688 0 1 1>}
  NamedRerouteDeclaration(None=@n5) {name=PushGradient, nodeColor=<rgba 0 1 0.1562 1>}
  NamedRerouteDeclaration(None=@n12.R) {name=ArmMask, nodeColor=<rgba 1 0.0296 0 1>}
  NamedRerouteDeclaration(None=@n12.G) {name=SleeveMask, nodeColor=<rgba 0.1551 1 0 1>}
  n14 = Power(
   Base=Add(A=@n4, B=ScalarParameter() {group=M_Ensemble_Snake_Arm, parameterName=MaterialCurveB, sliderMax=1})
   Exp=ScalarParameter() {defaultValue=16, group=M_Ensemble_Snake_Arm, parameterName=OpacityBand_Exp, sliderMax=32}
  )
  NamedRerouteDeclaration(None=OneMinus(None=Saturate(None=@n14))) {name=OpacityGradient, nodeColor=<rgba 0 1 0.7465 1>}
 ]}
