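# Physics asset for the dynamic skirt on the male medium Snake ensemble body, apparently a variant for sliding.
# Four skirt chains hang from the pelvis via free-rotation MSTR/AIM bones, with tight twist and limited swing down each chain.
# The two back chain tips are cross-linked by a constraint limiting one swing axis.
# Pelvis, thigh, calf and foot capsules provide body shapes, plus a large downward-offset root sphere.
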PhysicsAsset /BRCosmetics/Characters/Player/Male/Medium/Bodies/M_MED_Ensemble_Snake/Meshes/M_MED_Ensemble_dyn_skirt_sliding
bodies {
 calf_l [{
   shapeType=Capsule
   length=42.25
   radius=8.25
   shapeName=thigh_twist_01_l_capsule
   center=<-24.4076 -1.9425 0.5883>
   rotation=<rot 87.4373 0.3927 -4.0393>
  }]
 calf_r [{
   shapeType=Capsule
   length=42.25
   radius=8.25
   shapeName=thigh_twist_01_l_capsule
   center=<24.4076 1.9425 -0.5883>
   rotation=<rot 87.4373 0.3927 175.9607>
  }]
 dyn_skirt_bk_l_01 [
  {shapeType=Capsule, length=6.375, radius=7.75, shapeName=dyn_skirt_bk_l_01_capsule, center=<8 -2 6>, rotation=<rot -0.0003 0 95>}
  {shapeType=Capsule, length=7.75, radius=3.25, shapeName=dyn_skirt_bk_r_base_capsule, center=<8 8 -2>, rotation=<rot 80.0003 180 -175>}
 ]
 dyn_skirt_bk_l_02 [{shapeType=Capsule, length=7.75, radius=8.125, shapeName=dyn_skirt_bk_l_02_capsule, center=<8 -4 4>, rotation=<rot -0.0001 0 95>}]
 dyn_skirt_bk_l_MSTR [{shapeType=Capsule, length=0.505, radius=0.505, shapeName=dyn_skirt_bk_l_MSTR_capsule, rotation=<rot 0 -90 0>}]
 dyn_skirt_bk_l_base [
  {shapeType=Capsule, length=9.75, radius=7, shapeName=dyn_skirt_bk_l_01_capsule, center=<8 0 5>, rotation=<rot -0.0006 0 100>}
  {shapeType=Capsule, length=7.5, radius=3.5, shapeName=dyn_skirt_bk_r_base_capsule, center=<9 10 -5>, rotation=<rot 80.0001 180 -175>}
 ]
 dyn_skirt_bk_r_01 [
  {shapeType=Capsule, length=6.375, radius=7.75, shapeName=dyn_skirt_bk_l_01_capsule, center=<8 2 6>, rotation=<rot -0.0004 0 85>}
  {
   shapeType=Capsule
   length=7.75
   radius=3.25
   shapeName=dyn_skirt_bk_r_base_capsule
   center=<8 -8 -2>
   rotation=<rot 80.0005 -169.9995 -179.9995>
  }
 ]
 dyn_skirt_bk_r_02 [{shapeType=Capsule, length=7, radius=8.125, shapeName=dyn_skirt_bk_l_02_capsule, center=<8 4 4>, rotation=<rot -0.0001 0 85>}]
 dyn_skirt_bk_r_MSTR [{shapeType=Capsule, length=0.505, radius=0.505, shapeName=dyn_skirt_bk_r_MSTR_capsule, rotation=<rot 0 -90 0>}]
 dyn_skirt_bk_r_base [
  {shapeType=Capsule, length=10, radius=6.75, shapeName=dyn_skirt_bk_l_01_capsule, center=<8 0 5>, rotation=<rot -0.0014 0 75>}
  {
   shapeType=Capsule
   length=7.5
   radius=3.5
   shapeName=dyn_skirt_bk_r_base_capsule
   center=<9 -10 -5>
   rotation=<rot 81.9911 -173.7639 -178.7487>
  }
 ]
 dyn_skirt_ft_l_01 [{shapeType=Capsule, length=8.75, radius=3.875, shapeName=dyn_skirt_bk_l_01_capsule, center=<5 3 2>, rotation=<rot 0 0 90>}]
 dyn_skirt_ft_l_02 [{shapeType=Capsule, length=8.75, radius=3.875, shapeName=dyn_skirt_bk_l_01_capsule, center=<5 3 3>, rotation=<rot 0 0 90>}]
 dyn_skirt_ft_l_03 [{shapeType=Capsule, length=4.7, radius=8.75, shapeName=dyn_skirt_bk_l_02_capsule, center=<9 4 5>, rotation=<rot 0.0011 -0.0006 110>}]
 dyn_skirt_ft_l_04 [{shapeType=Capsule, length=8.625, radius=10, shapeName=dyn_skirt_bk_l_03_capsule, center=<10 7 5>, rotation=<rot 0 0 110>}]
 dyn_skirt_ft_l_AIM [{shapeType=Capsule, length=0.505, radius=0.505, shapeName=dyn_skirt_bk_r_MSTR_capsule, rotation=<rot 0 -90 0>}]
 dyn_skirt_ft_l_MSTR [{shapeType=Capsule, length=0.505, radius=0.505, shapeName=dyn_skirt_bk_r_MSTR_capsule, rotation=<rot 0 -90 0>}]
 dyn_skirt_ft_r_01 [{shapeType=Capsule, length=10.625, radius=4.125, shapeName=dyn_skirt_bk_l_01_capsule, center=<5 -3 2>, rotation=<rot 0.0001 0 95>}]
 dyn_skirt_ft_r_02 [{shapeType=Capsule, length=10.625, radius=4.125, shapeName=dyn_skirt_bk_l_01_capsule, center=<5 -3 3>, rotation=<rot 0.0001 0 95>}]
 dyn_skirt_ft_r_03 [{shapeType=Capsule, length=4.7, radius=8.75, shapeName=dyn_skirt_bk_l_02_capsule, center=<9 -4 5>, rotation=<rot 0.0011 -0.0006 69.9999>}]
 dyn_skirt_ft_r_04 [{shapeType=Capsule, length=8.625, radius=10, shapeName=dyn_skirt_bk_l_03_capsule, center=<10 -7 5>, rotation=<rot 0.0001 -0.0001 70>}]
 dyn_skirt_ft_r_AIM [{shapeType=Capsule, length=0.505, radius=0.505, shapeName=dyn_skirt_bk_r_MSTR_capsule, rotation=<rot 0 -90 0>}]
 dyn_skirt_ft_r_MSTR [{shapeType=Capsule, length=0.505, radius=0.505, shapeName=dyn_skirt_bk_r_MSTR_capsule, rotation=<rot 0 -90 0>}]
 foot_l [{shapeType=Capsule, length=18, radius=8.625, shapeName=thigh_twist_01_l_capsule, center=<-3.1879 7.1231 0>, rotation=<rot 0 0 90>}]
 foot_r [{shapeType=Capsule, length=18, radius=8.625, shapeName=thigh_twist_01_l_capsule, center=<3.1879 -7.1231 0>, rotation=<rot 0 0 -90>}]
 pelvis [{shapeType=Capsule, length=12.6862, radius=10.7732, shapeName=pelvis_capsule, center=<-7.3649 -0.413 0>}]
 root [{shapeType=Sphere, radius=80.375, shapeName=root_sphere, center=<0 2.1873 -58.9061>}]
 thigh_twist_01_l [{
   shapeType=Capsule
   length=24.557
   radius=8.6107
   shapeName=thigh_twist_01_l_capsule
   center=<3.0721 -0.5782 -1.4216>
   rotation=<rot 89.5637 1.281 0.0032>
  }]
 thigh_twist_01_r [{
   shapeType=Capsule
   length=24.432
   radius=8.7357
   shapeName=thigh_twist_01_l_capsule
   center=<-3.0782 0.578 0.6236>
   rotation=<rot 89.5637 1.281 -179.9968>
  }]
}
constraints [
 {bone1Name=dyn_skirt_bk_l_01, bone2Name=dyn_skirt_bk_l_base, swing1LimitDegrees=30, swing2LimitDegrees=45, twistLimitDegrees=5}
 {bone1Name=dyn_skirt_bk_l_02, bone2Name=dyn_skirt_bk_l_01, swing1LimitDegrees=30, swing2LimitDegrees=45, twistLimitDegrees=5}
 {bone1Name=dyn_skirt_bk_r_01, bone2Name=dyn_skirt_bk_r_base, swing1LimitDegrees=30, swing2LimitDegrees=45, twistLimitDegrees=5}
 {bone1Name=dyn_skirt_bk_r_02, bone2Name=dyn_skirt_bk_r_01, swing1LimitDegrees=30, swing2LimitDegrees=45, twistLimitDegrees=5}
 {
  bone1Name=dyn_skirt_bk_l_MSTR
  bone2Name=pelvis
  swing1LimitDegrees=45
  swing1Motion=Free
  swing2LimitDegrees=45
  swing2Motion=Free
  twistLimitDegrees=45
  twistMotion=Free
 }
 {
  bone1Name=dyn_skirt_bk_r_MSTR
  bone2Name=pelvis
  swing1LimitDegrees=45
  swing1Motion=Free
  swing2LimitDegrees=45
  swing2Motion=Free
  twistLimitDegrees=45
  twistMotion=Free
 }
 {bone1Name=dyn_skirt_bk_l_base, bone2Name=dyn_skirt_bk_l_MSTR, swing1LimitDegrees=30, swing2LimitDegrees=25, twistLimitDegrees=5}
 {bone1Name=dyn_skirt_bk_r_base, bone2Name=dyn_skirt_bk_r_MSTR, swing1LimitDegrees=30, swing2LimitDegrees=25, twistLimitDegrees=5}
 {
  bone1Name=calf_r
  bone2Name=pelvis
  swing1LimitDegrees=45
  swing1Motion=Free
  swing2LimitDegrees=45
  swing2Motion=Free
  twistLimitDegrees=45
  twistMotion=Free
 }
 {
  bone1Name=thigh_twist_01_l
  bone2Name=pelvis
  swing1LimitDegrees=45
  swing1Motion=Free
  swing2LimitDegrees=45
  swing2Motion=Free
  twistLimitDegrees=45
  twistMotion=Free
 }
 {
  bone1Name=dyn_skirt_ft_l_MSTR
  bone2Name=dyn_skirt_ft_l_AIM
  swing1LimitDegrees=45
  swing1Motion=Free
  swing2LimitDegrees=45
  swing2Motion=Free
  twistLimitDegrees=45
  twistMotion=Free
 }
 {bone1Name=dyn_skirt_ft_l_01, bone2Name=dyn_skirt_ft_l_MSTR, swing1LimitDegrees=10, swing2LimitDegrees=15, twistLimitDegrees=2}
 {bone1Name=dyn_skirt_ft_l_02, bone2Name=dyn_skirt_ft_l_01, swing1LimitDegrees=10, swing2LimitDegrees=15, twistLimitDegrees=2}
 {bone1Name=dyn_skirt_ft_l_03, bone2Name=dyn_skirt_ft_l_02, swing1LimitDegrees=10, swing2LimitDegrees=45, twistLimitDegrees=2}
 {bone1Name=dyn_skirt_ft_l_04, bone2Name=dyn_skirt_ft_l_03, swing1LimitDegrees=10, swing2LimitDegrees=45, twistLimitDegrees=2}
 {
  bone1Name=dyn_skirt_ft_r_MSTR
  bone2Name=dyn_skirt_ft_r_AIM
  swing1LimitDegrees=45
  swing1Motion=Free
  swing2LimitDegrees=45
  swing2Motion=Free
  twistLimitDegrees=45
  twistMotion=Free
 }
 {bone1Name=dyn_skirt_ft_r_01, bone2Name=dyn_skirt_ft_r_MSTR, swing1LimitDegrees=10, swing2LimitDegrees=15, twistLimitDegrees=2}
 {bone1Name=dyn_skirt_ft_r_02, bone2Name=dyn_skirt_ft_r_01, swing1LimitDegrees=10, swing2LimitDegrees=15, twistLimitDegrees=2}
 {bone1Name=dyn_skirt_ft_r_03, bone2Name=dyn_skirt_ft_r_02, swing1LimitDegrees=10, swing2LimitDegrees=45, twistLimitDegrees=2}
 {bone1Name=dyn_skirt_ft_r_04, bone2Name=dyn_skirt_ft_r_03, swing1LimitDegrees=10, swing2LimitDegrees=45, twistLimitDegrees=2}
 {
  bone1Name=dyn_skirt_ft_l_AIM
  bone2Name=pelvis
  swing1LimitDegrees=45
  swing1Motion=Free
  swing2LimitDegrees=45
  swing2Motion=Free
  twistLimitDegrees=45
  twistMotion=Free
 }
 {
  bone1Name=dyn_skirt_ft_r_AIM
  bone2Name=pelvis
  swing1LimitDegrees=45
  swing1Motion=Free
  swing2LimitDegrees=45
  swing2Motion=Free
  twistLimitDegrees=45
  twistMotion=Free
 }
 {
  bone1Name=thigh_twist_01_r
  bone2Name=pelvis
  swing1LimitDegrees=45
  swing1Motion=Free
  swing2LimitDegrees=45
  swing2Motion=Free
  twistLimitDegrees=45
  twistMotion=Free
 }
 {
  bone1Name=foot_l
  bone2Name=calf_l
  swing1LimitDegrees=45
  swing1Motion=Free
  swing2LimitDegrees=45
  swing2Motion=Free
  twistLimitDegrees=45
  twistMotion=Free
 }
 {
  bone1Name=foot_r
  bone2Name=calf_r
  swing1LimitDegrees=45
  swing1Motion=Free
  swing2LimitDegrees=45
  swing2Motion=Free
  twistLimitDegrees=45
  twistMotion=Free
 }
 {
  bone1Name=dyn_skirt_bk_r_02
  bone2Name=dyn_skirt_bk_l_02
  swing1LimitDegrees=10
  swing2LimitDegrees=45
  swing2Motion=Free
  twistLimitDegrees=45
  twistMotion=Free
 }
 {
  bone1Name=calf_l
  bone2Name=pelvis
  swing1LimitDegrees=45
  swing1Motion=Free
  swing2LimitDegrees=45
  swing2Motion=Free
  twistLimitDegrees=45
  twistMotion=Free
 }
 {
  bone1Name=pelvis
  bone2Name=root
  swing1LimitDegrees=45
  swing1Motion=Free
  swing2LimitDegrees=45
  swing2Motion=Free
  twistLimitDegrees=45
  twistMotion=Free
 }
]
