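# Physics asset for the dynamic ropes and bows on the medium male player's Snake ensemble.
# Two-segment left and right rope chains and single-capsule bows hang from small MSTR anchor bodies linked to a central knot, with side-specific swing limits and tight twist.
# Spine and arm capsules form a separate, freely rotating constraint chain that supplies body shapes around the cloth pieces.
# No constraint links the knot to the body chain.
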PhysicsAsset /Game/Characters/Player/Male/Medium/Bodies/M_MED_Ensemble_Snake/Meshes/M_MED_Ensemble_dyn_ropes
bodies {
 dyn_bow_l [{shapeType=Capsule, length=2.3789, radius=12.3057, shapeName=dyn_bow_l_capsule, center=<13 3.5 7.5>, rotation=<rot 0 0 84.9999>}]
 dyn_bow_l_MSTR [{shapeType=Capsule, length=1, radius=1, shapeName=dyn_knot_MSTR_capsule}]
 dyn_bow_r [{shapeType=Capsule, length=6.875, radius=11.75, shapeName=dyn_bow_l_capsule, center=<10 -3.5 13>, rotation=<rot 0 0 105.0002>}]
 dyn_bow_r_MSTR [{shapeType=Capsule, length=1, radius=1, shapeName=dyn_knot_MSTR_capsule}]
 dyn_knot_MSTR [{shapeType=Capsule, length=1, radius=1, shapeName=dyn_knot_MSTR_capsule}]
 dyn_ropes_l_01 [{shapeType=Capsule, length=11.75, radius=7.875, shapeName=dyn_ropes_r_01_capsule, center=<6 0 0>, rotation=<rot 90 146.3267 151.3269>}]
 dyn_ropes_l_02 [{shapeType=Capsule, length=9.125, radius=7.875, shapeName=dyn_ropes_r_01_capsule, center=<7 0 0>, rotation=<rot 90 0 0>}]
 dyn_ropes_l_MSTR [{shapeType=Capsule, length=1, radius=1, shapeName=dyn_knot_MSTR_capsule}]
 dyn_ropes_r_01 [{shapeType=Capsule, length=15.375, radius=3.75, shapeName=dyn_ropes_r_01_capsule, center=<8 0 0>, rotation=<rot 90 146.3267 156.3269>}]
 dyn_ropes_r_02 [{shapeType=Capsule, length=13.25, radius=3.75, shapeName=dyn_ropes_r_01_capsule, center=<8 0 0>, rotation=<rot 90 0 0>}]
 dyn_ropes_r_MSTR [{shapeType=Capsule, length=1, radius=1, shapeName=dyn_knot_MSTR_capsule}]
 lowerarm_l [{shapeType=Capsule, length=17.443, radius=4.4896, shapeName=upperarm_r_capsule, center=<10.0045 0.1194 0.1696>, rotation=<rot 90 0 180>}]
 lowerarm_r [{shapeType=Capsule, length=17.443, radius=4.4896, shapeName=upperarm_r_capsule, center=<-10.0045 -0.1194 -0.1696>, rotation=<rot 90 0 0>}]
 spine_05 [{shapeType=Capsule, length=3.3775, radius=15.5171, shapeName=spine_05_capsule, center=<2.0639 4.1112 0>}]
 upperarm_l [
  {shapeType=Capsule, length=10.193, radius=6.9896, shapeName=upperarm_r_capsule, center=<7.8556 -1.559 -0.6141>, rotation=<rot 90 0 180>}
  {shapeType=Capsule, length=17.193, radius=4.7396, shapeName=upperarm_r_capsule, center=<15.2512 -1.559 -0.6141>, rotation=<rot 90 0 180>}
 ]
 upperarm_r [{
   shapeType=Capsule
   length=21.193
   radius=6.7396
   shapeName=upperarm_r_capsule
   center=<-14.3353 0.8022 0.6141>
   rotation=<rot 90 -8.8807 -13.8809>
  }]
}
constraints [
 {bone1Name=dyn_ropes_r_01, bone2Name=dyn_ropes_r_MSTR, swing1LimitDegrees=35, swing2LimitDegrees=10, twistLimitDegrees=2}
 {bone1Name=dyn_bow_r, bone2Name=dyn_bow_r_MSTR, swing1LimitDegrees=45, swing2LimitDegrees=30, twistLimitDegrees=2}
 {bone1Name=dyn_bow_l, bone2Name=dyn_bow_l_MSTR, swing1LimitDegrees=45, swing2LimitDegrees=45, twistLimitDegrees=2}
 {bone1Name=dyn_ropes_r_02, bone2Name=dyn_ropes_r_01, swing1LimitDegrees=20, swing2LimitDegrees=10, twistLimitDegrees=2}
 {
  bone1Name=upperarm_r
  bone2Name=spine_05
  swing1LimitDegrees=45
  swing1Motion=Free
  swing2LimitDegrees=45
  swing2Motion=Free
  twistLimitDegrees=45
  twistMotion=Free
 }
 {
  bone1Name=lowerarm_r
  bone2Name=upperarm_r
  swing1LimitDegrees=45
  swing1Motion=Free
  swing2LimitDegrees=45
  swing2Motion=Free
  twistLimitDegrees=45
  twistMotion=Free
 }
 {
  bone1Name=lowerarm_l
  bone2Name=upperarm_l
  swing1LimitDegrees=45
  swing1Motion=Free
  swing2LimitDegrees=45
  swing2Motion=Free
  twistLimitDegrees=45
  twistMotion=Free
 }
 {
  bone1Name=upperarm_l
  bone2Name=spine_05
  swing1LimitDegrees=45
  swing1Motion=Free
  swing2LimitDegrees=45
  swing2Motion=Free
  twistLimitDegrees=45
  twistMotion=Free
 }
 {
  bone1Name=dyn_ropes_r_MSTR
  bone2Name=dyn_knot_MSTR
  swing1LimitDegrees=45
  swing1Motion=Free
  swing2LimitDegrees=45
  swing2Motion=Free
  twistLimitDegrees=45
  twistMotion=Free
 }
 {
  bone1Name=dyn_bow_l_MSTR
  bone2Name=dyn_knot_MSTR
  swing1LimitDegrees=45
  swing1Motion=Free
  swing2LimitDegrees=45
  swing2Motion=Free
  twistLimitDegrees=45
  twistMotion=Free
 }
 {
  bone1Name=dyn_bow_r_MSTR
  bone2Name=dyn_knot_MSTR
  swing1LimitDegrees=45
  swing1Motion=Free
  swing2LimitDegrees=45
  swing2Motion=Free
  twistLimitDegrees=45
  twistMotion=Free
 }
 {
  bone1Name=dyn_ropes_l_MSTR
  bone2Name=dyn_knot_MSTR
  swing1LimitDegrees=45
  swing1Motion=Free
  swing2LimitDegrees=45
  swing2Motion=Free
  twistLimitDegrees=45
  twistMotion=Free
 }
 {bone1Name=dyn_ropes_l_01, bone2Name=dyn_ropes_l_MSTR, swing1LimitDegrees=20, swing2LimitDegrees=30, twistLimitDegrees=2}
 {bone1Name=dyn_ropes_l_02, bone2Name=dyn_ropes_l_01, swing1LimitDegrees=20, swing2LimitDegrees=30, twistLimitDegrees=2}
]
