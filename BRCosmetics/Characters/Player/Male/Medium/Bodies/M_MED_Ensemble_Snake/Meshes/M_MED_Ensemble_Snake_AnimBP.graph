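# Animation Blueprint for the Snake male medium body cosmetic, adding secondary motion and corrective deformation to its skirt and sleeves.
# Thigh-rotation alphas drive additive skirt-bone offsets, feeding RigidBody simulations blended by sliding and sprinting.
# bRigidBodyEnabled switches between that simulation and a procedural skirt using foot-targeted Look Ats.
# A separate chain layers a montage-timed fangs emote, runs a deformation Control Rig, then simulates sleeves with AnimDynamics.
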AnimBlueprint /BRCosmetics/Characters/Player/Male/Medium/Bodies/M_MED_Ensemble_Snake/Meshes/M_MED_Ensemble_Snake_AnimBP
target_skeleton=&/M_MED_Ensemble_Snake_Skeleton
graphs [
 AnimGraph {nodes [
   n1 = "Transform (Modify) Bone - Bone: dyn_skirt_bk_l_MSTR"(
    ComponentPose=Local To Component()
    Alpha=Get thigh_rot_up_l() -> (thigh_rot_up_l:double)
    Translation="-12,0,-12"
    Rotation="90,0,0.000"
    Scale="1,1,1"
   ) {
    rotationMode=BMM_Additive
    rotationSpace=BCS_BoneSpace
    scaleMode=BMM_Additive
    scaleSpace=BCS_BoneSpace
    translationMode=BMM_Additive
    translationSpace=BCS_BoneSpace
    alphaBoolBlend {blendOption=Linear}
    boneToModify {boneName=dyn_skirt_bk_l_MSTR}
    rotation {pitch=90}
    translation {x=-12, z=-12}
   }
   n2 = "Transform (Modify) Bone - Bone: dyn_skirt_bk_l_MSTR"(ComponentPose=@n1, Alpha=Get thigh_rot_out_l() -> (thigh_rot_out_l:double), Translation="0,0,2", Rotation="-0.000,0,-30.000", Scale="1,1,1") {
    rotationMode=BMM_Additive
    rotationSpace=BCS_BoneSpace
    scaleMode=BMM_Additive
    scaleSpace=BCS_BoneSpace
    translationMode=BMM_Additive
    translationSpace=BCS_BoneSpace
    alphaBoolBlend {blendOption=Linear}
    boneToModify {boneName=dyn_skirt_bk_l_MSTR}
    rotation {roll=-30}
    translation {z=2}
   }
   n3 = "Transform (Modify) Bone - Bone: dyn_skirt_bk_r_MSTR"(ComponentPose=@n2, Alpha=Get thigh_rot_up_r() -> (thigh_rot_up_r:double), Translation="-12,0,-12", Rotation="90,0,0", Scale="1,1,1") {
    rotationMode=BMM_Additive
    rotationSpace=BCS_BoneSpace
    scaleMode=BMM_Additive
    scaleSpace=BCS_BoneSpace
    translationMode=BMM_Additive
    translationSpace=BCS_BoneSpace
    alphaBoolBlend {blendOption=Linear}
    boneToModify {boneName=dyn_skirt_bk_r_MSTR}
    rotation {pitch=90}
    translation {x=-12, z=-12}
   }
   n4 = "Transform (Modify) Bone - Bone: dyn_skirt_bk_r_MSTR"(ComponentPose=@n3, Alpha=Get thigh_rot_out_r() -> (thigh_rot_out_r:double), Translation="0,0,2", Rotation="0,0,30", Scale="1,1,1") {
    rotationMode=BMM_Additive
    rotationSpace=BCS_BoneSpace
    scaleMode=BMM_Additive
    scaleSpace=BCS_BoneSpace
    translationMode=BMM_Additive
    translationSpace=BCS_BoneSpace
    alphaBoolBlend {blendOption=Linear}
    boneToModify {boneName=dyn_skirt_bk_r_MSTR}
    rotation {roll=30}
    translation {z=2}
   }
   Component To Local(ComponentPose=@n4)
   n5 = "Constraint - dyn_skirt_ft_l_MSTR (P)"(Alpha=1, ConstraintWeights_0=1) {
    alphaBoolBlend {blendOption=Linear}
    boneToModify {boneName=dyn_skirt_ft_l_MSTR}
    constraintSetup [{targetBone {boneName=pelvis}}]
    constraintWeights [1]
   }
   "Constraint - dyn_skirt_ft_r_MSTR (P)"(ComponentPose=@n5, Alpha=1, ConstraintWeights_0=1) {
    alphaBoolBlend {blendOption=Linear}
    boneToModify {boneName=dyn_skirt_ft_r_MSTR}
    constraintSetup [{targetBone {boneName=pelvis}}]
    constraintWeights [1]
   }
   "Transform (Modify) Bone - Bone: dyn_ropes_r_MSTR"!(ComponentPose, bAlphaBoolEnabled:bool=True, Translation="0,0,0", Rotation="0.000,-20,-0.000", Scale="1,1,1") -> (Pose) {
    alphaInputType=Bool
    rotationMode=BMM_Additive
    rotationSpace=BCS_BoneSpace
    scaleMode=BMM_Additive
    scaleSpace=BCS_BoneSpace
    translationMode=BMM_Additive
    translationSpace=BCS_BoneSpace
    alphaBoolBlend {blendOption=Linear}
    boneToModify {boneName=dyn_ropes_r_MSTR}
    rotation {yaw=-20}
   }
   "Transform (Modify) Bone - Bone: dyn_knot_MSTR"!(ComponentPose, bAlphaBoolEnabled:bool=True, Translation="0,0,0", Rotation="0,0,0", Scale="1,1,1") -> (Pose) {
    alphaInputType=Bool
    rotationMode=BMM_Additive
    rotationSpace=BCS_BoneSpace
    scaleMode=BMM_Additive
    scaleSpace=BCS_BoneSpace
    translationMode=BMM_Additive
    translationSpace=BCS_BoneSpace
    alphaBoolBlend {blendOption=Linear}
    boneToModify {boneName=dyn_knot_MSTR}
   }
   RigidBody!(
    ComponentPose
    Alpha
    OverrideWorldGravity:Vector=Get GravityOverrideRope() -> (GravityOverrideRope:Vector)
    ExternalForce
    ComponentLinearAccScale:Vector="0.15,0.15,0.15"
   ) -> (Pose) {
    bClampLinearTranslationLimitToRefPose=true
    bOverrideWorldGravity=true
    overridePhysicsAsset=/Game/Characters/Player/Male/Medium/Bodies/M_MED_Ensemble_Snake/Meshes/M_MED_Ensemble_dyn_ropes
    simulationSpace=BaseBoneSpace
    alphaBoolBlend {blendOption=Linear}
    baseBoneRef {boneName=spine_02}
    componentAppliedLinearAccClamp=<1000 1000 1000>
    componentLinearAccScale=<0.3 0.3 0.2>
    componentLinearVelScale=<0.2 0.2 0.2>
    overrideWorldGravity {z=-980}
    simSpaceSettings {
     maxAngularAcceleration=200
     maxAngularVelocity=5
     maxLinearAcceleration=1000
     maxLinearVelocity=200
     velocityScaleZ=0
     externalLinearDragV=<1 1>
    }
   }
   n6 = "Constraint - dyn_skirt_ft_l_MSTR (P)"(ComponentPose=Local To Component(LocalPose=Use cached pose 'preRigidPose'()), Alpha=1, ConstraintWeights_0=1) {
    alphaBoolBlend {blendOption=Linear}
    boneToModify {boneName=dyn_skirt_ft_l_MSTR}
    constraintSetup [{targetBone {boneName=pelvis}}]
    constraintWeights [1]
   }
   n7 = "Constraint - dyn_skirt_ft_r_MSTR (P)"(ComponentPose=@n6, Alpha=1, ConstraintWeights_0=1) {
    alphaBoolBlend {blendOption=Linear}
    boneToModify {boneName=dyn_skirt_ft_r_MSTR}
    constraintSetup [{targetBone {boneName=pelvis}}]
    constraintWeights [1]
   }
   n8 = "Constraint - dyn_skirt_bk_l_MSTR (P)"(ComponentPose=@n7, Alpha=1, ConstraintWeights_0=1) {
    alphaBoolBlend {blendOption=Linear}
    boneToModify {boneName=dyn_skirt_bk_l_MSTR}
    constraintSetup [{targetBone {boneName=pelvis}}]
    constraintWeights [1]
   }
   n9 = "Constraint - dyn_skirt_bk_r_MSTR (P)"(ComponentPose=@n8, Alpha=1, ConstraintWeights_0=1) {
    alphaBoolBlend {blendOption=Linear}
    boneToModify {boneName=dyn_skirt_bk_r_MSTR}
    constraintSetup [{targetBone {boneName=pelvis}}]
    constraintWeights [1]
   }
   n10 = Get thigh_rot_up_l()
   n11 = "Transform (Modify) Bone - Bone: dyn_skirt_ft_l_MSTR"(ComponentPose=@n9, Alpha=@n10, Translation="-8,-3,1", Rotation="20,10,0", Scale="1,1,1") {
    rotationMode=BMM_Additive
    rotationSpace=BCS_BoneSpace
    translationMode=BMM_Additive
    translationSpace=BCS_BoneSpace
    alphaBoolBlend {blendOption=Linear}
    boneToModify {boneName=dyn_skirt_ft_l_MSTR}
    rotation {pitch=20, yaw=10}
    translation=<-8 -3 1>
   }
   n12 = "Transform (Modify) Bone - Bone: dyn_skirt_ft_mid_MSTR"(ComponentPose=@n11, Alpha=@n10, Translation="-1,0,0", Rotation="6,0.000,0.000", Scale="1,1,1") {
    rotationMode=BMM_Additive
    rotationSpace=BCS_BoneSpace
    scaleMode=BMM_Additive
    scaleSpace=BCS_BoneSpace
    translationMode=BMM_Additive
    translationSpace=BCS_BoneSpace
    alphaBoolBlend {blendOption=Linear}
    boneToModify {boneName=dyn_skirt_ft_mid_MSTR}
    rotation {pitch=6}
    translation {x=-1}
   }
   n13 = Get thigh_rot_up_r()
   n14 = "Transform (Modify) Bone - Bone: dyn_skirt_ft_r_MSTR"(ComponentPose=@n12, Alpha=@n13, Translation="-8,3,1", Rotation="20,-10,0", Scale="1,1,1") {
    rotationMode=BMM_Additive
    rotationSpace=BCS_BoneSpace
    translationMode=BMM_Additive
    translationSpace=BCS_BoneSpace
    alphaBoolBlend {blendOption=Linear}
    boneToModify {boneName=dyn_skirt_ft_r_MSTR}
    rotation {pitch=20, yaw=-10}
    translation=<-8 3 1>
   }
   n15 = "Transform (Modify) Bone - Bone: dyn_skirt_ft_mid_MSTR"(ComponentPose=@n14, Alpha=@n13, Translation="-1,0,0", Rotation="6,0,0", Scale="1,1,1") {
    rotationMode=BMM_Additive
    rotationSpace=BCS_BoneSpace
    scaleMode=BMM_Additive
    scaleSpace=BCS_BoneSpace
    translationMode=BMM_Additive
    translationSpace=BCS_BoneSpace
    alphaBoolBlend {blendOption=Linear}
    boneToModify {boneName=dyn_skirt_ft_mid_MSTR}
    rotation {pitch=6}
    translation {x=-1}
   }
   n16 = "Transform (Modify) Bone - Bone: dyn_skirt_ft_mid_MSTR"(ComponentPose=@n15, Alpha=1, Translation="-5,0,0", Rotation="40,0.000,0.000", Scale="1,1,1") {
    rotationMode=BMM_Additive
    rotationSpace=BCS_BoneSpace
    translationMode=BMM_Additive
    translationSpace=BCS_BoneSpace
    alphaBoolBlend {blendOption=Linear}
    boneToModify {boneName=dyn_skirt_ft_mid_MSTR}
    rotation {pitch=40}
    translation {x=-5}
   }
   n17 = "Transform (Modify) Bone - Bone: dyn_skirt_ft_l_MSTR"(ComponentPose=@n16, Alpha=1, Translation="0,0,0", Rotation="30,30,0", Scale="1,1,1") {
    rotationMode=BMM_Additive
    rotationSpace=BCS_BoneSpace
    translationMode=BMM_Additive
    translationSpace=BCS_BoneSpace
    alphaBoolBlend {blendOption=Linear}
    boneToModify {boneName=dyn_skirt_ft_l_MSTR}
    rotation {pitch=30, yaw=30}
   }
   n18 = "Transform (Modify) Bone - Bone: dyn_skirt_ft_r_MSTR"(ComponentPose=@n17, Alpha=1, Translation="0,0,0", Rotation="30,-30,-0.000", Scale="1,1,1") {
    rotationMode=BMM_Additive
    rotationSpace=BCS_BoneSpace
    translationMode=BMM_Additive
    translationSpace=BCS_BoneSpace
    alphaBoolBlend {blendOption=Linear}
    boneToModify {boneName=dyn_skirt_ft_r_MSTR}
    rotation {pitch=30, yaw=-30}
   }
   n19 = "Transform (Modify) Bone - Bone: dyn_skirt_bk_l_MSTR"(ComponentPose=@n18, Alpha=1, Translation="0,0,5", Rotation="30,0,0", Scale="1,1,1") {
    rotationMode=BMM_Additive
    rotationSpace=BCS_BoneSpace
    translationMode=BMM_Additive
    translationSpace=BCS_BoneSpace
    alphaBoolBlend {blendOption=Linear}
    boneToModify {boneName=dyn_skirt_bk_l_MSTR}
    rotation {pitch=30}
    translation {z=5}
   }
   n20 = "Transform (Modify) Bone - Bone: dyn_skirt_bk_r_MSTR"(ComponentPose=@n19, Alpha=1, Translation="0,0,5", Rotation="30,0,0", Scale="1,1,1") {
    rotationMode=BMM_Additive
    rotationSpace=BCS_BoneSpace
    translationMode=BMM_Additive
    translationSpace=BCS_BoneSpace
    alphaBoolBlend {blendOption=Linear}
    boneToModify {boneName=dyn_skirt_bk_r_MSTR}
    rotation {pitch=30}
    translation {z=5}
   }
   n21 = RigidBody!(ComponentPose=@n20, Alpha, OverrideWorldGravity="0,0,-980", ExternalForce) -> (Pose) {
    bClampLinearTranslationLimitToRefPose=true
    bOverrideWorldGravity=true
    overridePhysicsAsset=&/M_MED_Ensemble_dyn_skirt
    simulationSpace=BaseBoneSpace
    alphaBoolBlend {blendOption=Linear}
    baseBoneRef {boneName=spine_01}
    componentAppliedLinearAccClamp=<1000 1000 1000>
    componentLinearAccScale=<0.2 0.2 0.2>
    componentLinearVelScale=<0.5 0.5 0.5>
    overrideWorldGravity {z=-980}
    simSpaceSettings {
     maxAngularAcceleration=200
     maxAngularVelocity=5
     maxLinearAcceleration=1000
     maxLinearVelocity=200
     velocityScaleZ=0
     externalLinearDragV=<1 1>
    }
   }
   n22 = "Transform (Modify) Bone - Bone: dyn_skirt_ft_l_MSTR"(ComponentPose=@n21, Alpha=1, Translation="0,3,-0.000", Rotation="5,5,0.000", Scale="1,1,1") {
    rotationMode=BMM_Additive
    rotationSpace=BCS_BoneSpace
    translationMode=BMM_Additive
    translationSpace=BCS_BoneSpace
    alphaBoolBlend {blendOption=Linear}
    boneToModify {boneName=dyn_skirt_ft_l_MSTR}
    rotation {pitch=5, yaw=5}
    translation {y=3}
   }
   n23 = "Transform (Modify) Bone - Bone: dyn_skirt_ft_r_MSTR"(ComponentPose=@n22, Alpha=1, Translation="0,-3,-0.000", Rotation="5,-5,0", Scale="1,1,1") {
    rotationMode=BMM_Additive
    rotationSpace=BCS_BoneSpace
    translationMode=BMM_Additive
    translationSpace=BCS_BoneSpace
    alphaBoolBlend {blendOption=Linear}
    boneToModify {boneName=dyn_skirt_ft_r_MSTR}
    rotation {pitch=5, yaw=-5}
    translation {y=-3}
   }
   n24 = "Transform (Modify) Bone - Bone: dyn_skirt_ft_mid_MSTR"(ComponentPose=@n23, Alpha=1, Translation="0,0,0", Rotation="10,0.000,0.000", Scale="1,1,1") {
    rotationMode=BMM_Additive
    rotationSpace=BCS_BoneSpace
    translationMode=BMM_Additive
    translationSpace=BCS_BoneSpace
    alphaBoolBlend {blendOption=Linear}
    boneToModify {boneName=dyn_skirt_ft_mid_MSTR}
    rotation {pitch=10}
   }
   n25 = Component To Local(ComponentPose=@n24)
   n26 = "Constraint - dyn_skirt_ft_mid_TGT (PP)"!(
    ComponentPose=Local To Component(LocalPose=Use cached pose 'preRigidPose'())
    Alpha=1
    ConstraintWeights_0=0.5
    ConstraintWeights_1:float=0.5
   ) -> (Pose) {
    alphaBoolBlend {blendOption=Linear}
    boneToModify {boneName=dyn_skirt_ft_mid_TGT}
    constraintSetup [{targetBone {boneName=thigh_twist_01_l}}, {targetBone {boneName=thigh_twist_01_r}}]
    constraintWeights [1,1]
   }
   n27 = Look At - Bone: dyn_skirt_ft_mid_AIM(ComponentPose=@n26, Alpha=1) {
    alphaBoolBlend {blendOption=Linear}
    boneToModify {boneName=dyn_skirt_ft_mid_AIM}
    lookAtLocation {x=0}
    lookAtTarget {boneReference {boneName=dyn_skirt_ft_mid_TGT}}
    lookAt_Axis {axis=<1 0>}
   }
   n28 = Get thigh_rot_up_l()
   n29 = "Transform (Modify) Bone - Bone: dyn_skirt_ft_l_MSTR"(ComponentPose=@n27, Alpha=@n28, Translation="20,3,3", Rotation="-40,15,40", Scale="1,1,1") {
    rotationMode=BMM_Additive
    rotationSpace=BCS_BoneSpace
    translationMode=BMM_Additive
    translationSpace=BCS_BoneSpace
    alphaBoolBlend {blendOption=Linear}
    boneToModify {boneName=dyn_skirt_ft_l_MSTR}
    rotation=<rot -40 15 40>
    translation=<20 3 3>
   }
   n30 = "Transform (Modify) Bone - Bone: dyn_skirt_ft_mid_MSTR"(ComponentPose=@n29, Alpha=@n28, Translation="-5,0,2.342", Rotation="0,0,0", Scale="1,1,1") {
    rotationMode=BMM_Additive
    rotationSpace=BCS_BoneSpace
    scaleMode=BMM_Additive
    scaleSpace=BCS_BoneSpace
    translationMode=BMM_Additive
    translationSpace=BCS_BoneSpace
    alphaBoolBlend {blendOption=Linear}
    boneToModify {boneName=dyn_skirt_ft_mid_MSTR}
    translation {x=-5, z=2.342}
   }
   n31 = "Transform (Modify) Bone - Bone: dyn_skirt_ft_l_MSTR"(ComponentPose=@n30, Alpha=Get thigh_rot_out_l(), Translation="-0.000,-1,2", Rotation="0,0,0", Scale="1,1,1") {
    rotationMode=BMM_Additive
    rotationSpace=BCS_BoneSpace
    translationMode=BMM_Additive
    translationSpace=BCS_BoneSpace
    alphaBoolBlend {blendOption=Linear}
    boneToModify {boneName=dyn_skirt_ft_l_MSTR}
    translation {y=-1, z=2}
   }
   n32 = "Transform (Modify) Bone - Bone: dyn_skirt_ft_r_MSTR"(ComponentPose=@n31, Alpha=Get thigh_rot_up_r(), Translation="20,-3.768,6", Rotation="-50,-10,-25.316", Scale="1,1,1") {
    rotationMode=BMM_Additive
    rotationSpace=BCS_BoneSpace
    translationMode=BMM_Additive
    translationSpace=BCS_BoneSpace
    alphaBoolBlend {blendOption=Linear}
    boneToModify {boneName=dyn_skirt_ft_r_MSTR}
    rotation=<rot -50 -10 -25.316>
    translation=<20 -3.768 6>
   }
   n33 = "Transform (Modify) Bone - Bone: dyn_skirt_ft_mid_MSTR"(ComponentPose=@n32, Alpha=0, Translation="0,8,-0.000", Rotation="0,0,0", Scale="1,1,1") {
    rotationMode=BMM_Additive
    rotationSpace=BCS_BoneSpace
    scaleMode=BMM_Additive
    scaleSpace=BCS_BoneSpace
    translationMode=BMM_Additive
    translationSpace=BCS_BoneSpace
    alphaBoolBlend {blendOption=Linear}
    boneToModify {boneName=dyn_skirt_ft_mid_MSTR}
    translation {y=8}
   }
   n34 = "Transform (Modify) Bone - Bone: dyn_skirt_ft_r_MSTR"(ComponentPose=@n33, Alpha=Get thigh_rot_in_l() -> (thigh_rot_in_l:double), Translation="0,2,1", Rotation="0,0,0", Scale="1,1,1") {
    rotationMode=BMM_Additive
    rotationSpace=BCS_BoneSpace
    translationMode=BMM_Additive
    translationSpace=BCS_BoneSpace
    alphaBoolBlend {blendOption=Linear}
    boneToModify {boneName=dyn_skirt_ft_r_MSTR}
    translation {y=2, z=1}
   }
   n35 = "Transform (Modify) Bone - Bone: dyn_skirt_bk_l_MSTR"(ComponentPose=@n34, Alpha=Get thigh_rot_up_l(), Translation="-12,0,-13", Rotation="100,0,0.000", Scale="1,1,1") {
    rotationMode=BMM_Additive
    rotationSpace=BCS_BoneSpace
    scaleMode=BMM_Additive
    scaleSpace=BCS_BoneSpace
    translationMode=BMM_Additive
    translationSpace=BCS_BoneSpace
    alphaBoolBlend {blendOption=Linear}
    boneToModify {boneName=dyn_skirt_bk_l_MSTR}
    rotation {pitch=100}
    translation {x=-12, z=-13}
   }
   n36 = "Transform (Modify) Bone - Bone: dyn_skirt_bk_l_MSTR"(ComponentPose=@n35, Alpha=Get thigh_rot_out_l(), Translation="0,0,2", Rotation="-0.000,0,-30.000", Scale="1,1,1") {
    rotationMode=BMM_Additive
    rotationSpace=BCS_BoneSpace
    scaleMode=BMM_Additive
    scaleSpace=BCS_BoneSpace
    translationMode=BMM_Additive
    translationSpace=BCS_BoneSpace
    alphaBoolBlend {blendOption=Linear}
    boneToModify {boneName=dyn_skirt_bk_l_MSTR}
    rotation {roll=-30}
    translation {z=2}
   }
   n37 = "Transform (Modify) Bone - Bone: dyn_skirt_bk_r_MSTR"(ComponentPose=@n36, Alpha=Get thigh_rot_up_r(), Translation="-12,0,-13", Rotation="100,0,0", Scale="1,1,1") {
    rotationMode=BMM_Additive
    rotationSpace=BCS_BoneSpace
    scaleMode=BMM_Additive
    scaleSpace=BCS_BoneSpace
    translationMode=BMM_Additive
    translationSpace=BCS_BoneSpace
    alphaBoolBlend {blendOption=Linear}
    boneToModify {boneName=dyn_skirt_bk_r_MSTR}
    rotation {pitch=100}
    translation {x=-12, z=-13}
   }
   n38 = "Transform (Modify) Bone - Bone: dyn_skirt_bk_r_MSTR"(ComponentPose=@n37, Alpha=Get thigh_rot_out_r(), Translation="0,0,2", Rotation="0,0,30", Scale="1,1,1") {
    rotationMode=BMM_Additive
    rotationSpace=BCS_BoneSpace
    scaleMode=BMM_Additive
    scaleSpace=BCS_BoneSpace
    translationMode=BMM_Additive
    translationSpace=BCS_BoneSpace
    alphaBoolBlend {blendOption=Linear}
    boneToModify {boneName=dyn_skirt_bk_r_MSTR}
    rotation {roll=30}
    translation {z=2}
   }
   n39 = "Transform (Modify) Bone - Bone: dyn_skirt_bk_r_MSTR"!(ComponentPose=@n38, bAlphaBoolEnabled:bool=True, Translation="0,0,0", Rotation="30,0,0", Scale="1,1,1") -> (Pose) {
    alphaInputType=Bool
    rotationMode=BMM_Additive
    rotationSpace=BCS_BoneSpace
    translationMode=BMM_Additive
    translationSpace=BCS_BoneSpace
    alphaBoolBlend {blendOption=Linear}
    boneToModify {boneName=dyn_skirt_bk_r_MSTR}
    rotation {pitch=30}
   }
   n40 = "Transform (Modify) Bone - Bone: dyn_skirt_bk_l_MSTR"!(ComponentPose=@n39, bAlphaBoolEnabled:bool=True, Translation="0,0,0", Rotation="30,0,0", Scale="1,1,1") -> (Pose) {
    alphaInputType=Bool
    rotationMode=BMM_Additive
    rotationSpace=BCS_BoneSpace
    translationMode=BMM_Additive
    translationSpace=BCS_BoneSpace
    alphaBoolBlend {blendOption=Linear}
    boneToModify {boneName=dyn_skirt_bk_l_MSTR}
    rotation {pitch=30}
   }
   n41 = Get bIsCrouchSprinting()
   n42 = "Transform (Modify) Bone - Bone: dyn_skirt_ft_r_MSTR"!(ComponentPose=@n40, bAlphaBoolEnabled:bool=@n41, Translation="0,0,0", Rotation="30,0,0", Scale="1,1,1") -> (Pose) {
    alphaInputType=Bool
    rotationMode=BMM_Additive
    rotationSpace=BCS_BoneSpace
    translationMode=BMM_Additive
    translationSpace=BCS_BoneSpace
    alphaBoolBlend {blendOption=Linear}
    boneToModify {boneName=dyn_skirt_ft_r_MSTR}
    rotation {pitch=30}
   }
   n43 = "Transform (Modify) Bone - Bone: dyn_skirt_ft_l_MSTR"!(ComponentPose=@n42, bAlphaBoolEnabled:bool=@n41, Translation="0,0,0", Rotation="30,0,0", Scale="1,1,1") -> (Pose) {
    alphaInputType=Bool
    rotationMode=BMM_Additive
    rotationSpace=BCS_BoneSpace
    translationMode=BMM_Additive
    translationSpace=BCS_BoneSpace
    alphaBoolBlend {blendOption=Linear}
    boneToModify {boneName=dyn_skirt_ft_l_MSTR}
    rotation {pitch=30}
   }
   n44 = RigidBody(ComponentPose=@n43, OverrideWorldGravity=Get GravityOverride()) {
    bClampLinearTranslationLimitToRefPose=true
    bOverrideWorldGravity=true
    overridePhysicsAsset=&M_MED_Ensemble_dyn_skirt
    simulationSpace=BaseBoneSpace
    alphaBoolBlend {blendOption=Linear}
    baseBoneRef {boneName=pelvis}
    componentAppliedLinearAccClamp=<3000 3000 3000>
    componentLinearAccScale=<0.3 0.3 0.3>
    componentLinearVelScale=<0.5 0.5 0.5>
    overrideWorldGravity {z=-980}
    simSpaceSettings {
     maxAngularAcceleration=200
     maxAngularVelocity=5
     maxLinearAcceleration=1000
     maxLinearVelocity=200
     velocityScaleZ=0
     externalLinearDragV=<1 1>
    }
   }
   n45 = "Transform (Modify) Bone - Bone: dyn_skirt_ft_l_MSTR"!(ComponentPose=@n44, Alpha:float=Get move_out_robe_l() -> (move_out_robe_l:double), Translation="0,2,3", Rotation="20,5,-10", Scale="1,1,1") -> (Pose) {
    rotationMode=BMM_Additive
    rotationSpace=BCS_BoneSpace
    translationMode=BMM_Additive
    translationSpace=BCS_BoneSpace
    alphaBoolBlend {blendOption=Linear}
    boneToModify {boneName=dyn_skirt_ft_l_MSTR}
    rotation=<rot 20 5 -10>
    translation {y=2, z=3}
   }
   n46 = "Transform (Modify) Bone - Bone: dyn_skirt_ft_r_MSTR"!(ComponentPose=@n45, Alpha:float=Get move_out_robe_r() -> (move_out_robe_r:double), Translation="0,-2,3", Rotation="20,-5,0", Scale="1,1,1") -> (Pose) {
    rotationMode=BMM_Additive
    rotationSpace=BCS_BoneSpace
    translationMode=BMM_Additive
    translationSpace=BCS_BoneSpace
    alphaBoolBlend {blendOption=Linear}
    boneToModify {boneName=dyn_skirt_ft_r_MSTR}
    rotation {pitch=20, yaw=-5}
    translation {y=-2, z=3}
   }
   n47 = Component To Local(ComponentPose=@n46)
   n48 = Get sprinting_alpha() -> (sprinting_alpha:bool)
   n49 = "Constraint - dyn_skirt_bk_l_MSTR (P)"(ComponentPose=Local To Component(LocalPose=Use cached pose 'preRigidPose'()), Alpha=1, ConstraintWeights_0=1) {
    alphaBoolBlend {blendOption=Linear}
    boneToModify {boneName=dyn_skirt_bk_l_MSTR}
    constraintSetup [{targetBone {boneName=spine_02}}]
    constraintWeights [1]
   }
   n50 = "Constraint - dyn_skirt_bk_r_MSTR (P)"(ComponentPose=@n49, Alpha=1, ConstraintWeights_0=1) {
    alphaBoolBlend {blendOption=Linear}
    boneToModify {boneName=dyn_skirt_bk_r_MSTR}
    constraintSetup [{targetBone {boneName=spine_02}}]
    constraintWeights [1]
   }
   n51 = "Transform (Modify) Bone - Bone: dyn_skirt_ft_mid_MSTR"(ComponentPose=@n50, Alpha=0, Translation="0,0,0", Rotation="60.000,0.000,0.000", Scale="1,1,1") {
    rotationMode=BMM_Additive
    rotationSpace=BCS_BoneSpace
    translationMode=BMM_Additive
    translationSpace=BCS_BoneSpace
    alphaBoolBlend {blendOption=Linear}
    boneToModify {boneName=dyn_skirt_ft_mid_MSTR}
    rotation {pitch=60}
   }
   n52 = "Transform (Modify) Bone - Bone: dyn_skirt_ft_l_MSTR"(ComponentPose=@n51, Alpha=0, Translation="0,0,0", Rotation="40,60,60", Scale="1,1,1") {
    rotationMode=BMM_Additive
    rotationSpace=BCS_BoneSpace
    translationMode=BMM_Additive
    translationSpace=BCS_BoneSpace
    alphaBoolBlend {blendOption=Linear}
    boneToModify {boneName=dyn_skirt_ft_l_MSTR}
    rotation=<rot 40 60 60>
   }
   n53 = "Transform (Modify) Bone - Bone: dyn_skirt_ft_r_MSTR"(ComponentPose=@n52, Alpha=0, Translation="0,0,0", Rotation="40,-60,-60", Scale="1,1,1") {
    rotationMode=BMM_Additive
    rotationSpace=BCS_BoneSpace
    translationMode=BMM_Additive
    translationSpace=BCS_BoneSpace
    alphaBoolBlend {blendOption=Linear}
    boneToModify {boneName=dyn_skirt_ft_r_MSTR}
    rotation=<rot 40 -60 -60>
   }
   n54 = "Transform (Modify) Bone - Bone: dyn_skirt_bk_l_MSTR"!(ComponentPose=@n53, Alpha:float=1, Translation="-15,0,0", Rotation="60,0,0", Scale="1,1,1") -> (Pose) {
    rotationMode=BMM_Additive
    rotationSpace=BCS_BoneSpace
    translationMode=BMM_Additive
    translationSpace=BCS_BoneSpace
    alphaBoolBlend {blendOption=Linear}
    boneToModify {boneName=dyn_skirt_bk_l_MSTR}
    rotation {pitch=60}
    translation {x=-15}
   }
   n55 = "Transform (Modify) Bone - Bone: dyn_skirt_bk_r_MSTR"!(ComponentPose=@n54, Alpha:float=1, Translation="-15,0,0", Rotation="60,0,0", Scale="1,1,1") -> (Pose) {
    rotationMode=BMM_Additive
    rotationSpace=BCS_BoneSpace
    translationMode=BMM_Additive
    translationSpace=BCS_BoneSpace
    alphaBoolBlend {blendOption=Linear}
    boneToModify {boneName=dyn_skirt_bk_r_MSTR}
    rotation {pitch=60}
    translation {x=-15}
   }
   n56 = RigidBody(ComponentPose=@n55, OverrideWorldGravity="0,0,-980") {
    bClampLinearTranslationLimitToRefPose=true
    bOverrideWorldGravity=true
    overridePhysicsAsset=&/M_MED_Ensemble_dyn_skirt_sliding
    simulationSpace=BaseBoneSpace
    alphaBoolBlend {blendOption=Linear}
    baseBoneRef {boneName=spine_02}
    componentAppliedLinearAccClamp=<3000 3000 3000>
    componentLinearAccScale=<0.3 0.3 0.3>
    componentLinearVelScale=<0.5 0.5 0.5>
    overrideWorldGravity {z=-980}
    simSpaceSettings {
     maxAngularAcceleration=200
     maxAngularVelocity=5
     maxLinearAcceleration=1000
     maxLinearVelocity=200
     velocityScaleZ=0
     externalLinearDragV=<1 1>
    }
   }
   n57 = Local To Component(LocalPose=Blend Poses by bool(
     BlendPose_0=Component To Local(ComponentPose=@n56)
     BlendPose_1=Blend Poses by bool(BlendPose_0=@n25, BlendPose_1=@n47, BlendTime_0=0.35, bActiveValue=@n48) {blendType=Linear}
     BlendTime_0=0.8
     BlendTime_1=0.2
     bActiveValue=Get bIsSliding()
    ) {blendType=Linear})
   n58 = "Constraint - dyn_skirt_ft_mid_TGT (PP)"(ComponentPose=Local To Component(LocalPose=Use cached pose 'basePose'()), Alpha=1, ConstraintWeights_0=0.5, ConstraintWeights_1=0.5) {
    alphaBoolBlend {blendOption=Linear}
    boneToModify {boneName=dyn_skirt_ft_mid_TGT}
    constraintSetup [{targetBone {boneName=thigh_twist_01_l}}, {targetBone {boneName=thigh_twist_01_r}}]
    constraintWeights [1,1]
   }
   n59 = Look At - Bone: dyn_skirt_ft_mid_AIM(ComponentPose=@n58, Alpha=1) {
    alphaBoolBlend {blendOption=Linear}
    boneToModify {boneName=dyn_skirt_ft_mid_AIM}
    lookAtLocation {x=0}
    lookAtTarget {boneReference {boneName=dyn_skirt_ft_mid_TGT}}
    lookAt_Axis {axis=<1 0>}
   }
   n60 = Get thigh_rot_up_l()
   n61 = "Transform (Modify) Bone - Bone: dyn_skirt_ft_l_MSTR"(ComponentPose=@n59, Alpha=@n60, Translation="13.269,3.582,3.321", Rotation="-8.030,-1.629,8.424", Scale="0.750,1.000,1.000") {
    rotationMode=BMM_Additive
    rotationSpace=BCS_BoneSpace
    scaleMode=BMM_Additive
    scaleSpace=BCS_BoneSpace
    translationMode=BMM_Additive
    translationSpace=BCS_BoneSpace
    alphaBoolBlend {blendOption=Linear}
    boneToModify {boneName=dyn_skirt_ft_l_MSTR}
    rotation=<rot -8.03 -1.629 8.424>
    scale {x=0.75}
    translation=<13.269 3.582 3.321>
   }
   n62 = "Transform (Modify) Bone - Bone: dyn_skirt_ft_mid_MSTR"(ComponentPose=@n61, Alpha=@n60, Translation="-5.000,0.000,2.342", Rotation="0,0,0", Scale="0.750,1.000,1.000") {
    rotationMode=BMM_Additive
    rotationSpace=BCS_BoneSpace
    scaleMode=BMM_Additive
    scaleSpace=BCS_BoneSpace
    translationMode=BMM_Additive
    translationSpace=BCS_BoneSpace
    alphaBoolBlend {blendOption=Linear}
    boneToModify {boneName=dyn_skirt_ft_mid_MSTR}
    scale {x=0.75}
    translation {x=-5, z=2.342}
   }
   n63 = "Transform (Modify) Bone - Bone: dyn_skirt_ft_r_MSTR"(ComponentPose=@n62, Alpha=Get thigh_rot_up_r(), Translation="13,-3.768,3", Rotation="-10,0,-8", Scale="0.75,1,1") {
    rotationMode=BMM_Additive
    rotationSpace=BCS_BoneSpace
    scaleMode=BMM_Additive
    scaleSpace=BCS_BoneSpace
    translationMode=BMM_Additive
    translationSpace=BCS_BoneSpace
    alphaBoolBlend {blendOption=Linear}
    boneToModify {boneName=dyn_skirt_ft_r_MSTR}
    rotation {pitch=-10, roll=-8}
    scale {x=0.75}
    translation=<13 -3.768 3>
   }
   n64 = "Transform (Modify) Bone - Bone: dyn_skirt_bk_l_MSTR"(ComponentPose=@n63, Alpha=Get thigh_rot_up_l(), Translation="-12,0,-13", Rotation="100,0,0.000", Scale="1,1,1") {
    rotationMode=BMM_Additive
    rotationSpace=BCS_BoneSpace
    scaleMode=BMM_Additive
    scaleSpace=BCS_BoneSpace
    translationMode=BMM_Additive
    translationSpace=BCS_BoneSpace
    alphaBoolBlend {blendOption=Linear}
    boneToModify {boneName=dyn_skirt_bk_l_MSTR}
    rotation {pitch=100}
    translation {x=-12, z=-13}
   }
   n65 = "Transform (Modify) Bone - Bone: dyn_skirt_bk_l_MSTR"(ComponentPose=@n64, Alpha=Get thigh_rot_out_l(), Translation="0,0,2", Rotation="-0.000,0,-30.000", Scale="1,1,1") {
    rotationMode=BMM_Additive
    rotationSpace=BCS_BoneSpace
    scaleMode=BMM_Additive
    scaleSpace=BCS_BoneSpace
    translationMode=BMM_Additive
    translationSpace=BCS_BoneSpace
    alphaBoolBlend {blendOption=Linear}
    boneToModify {boneName=dyn_skirt_bk_l_MSTR}
    rotation {roll=-30}
    translation {z=2}
   }
   n66 = "Transform (Modify) Bone - Bone: dyn_skirt_bk_r_MSTR"(ComponentPose=@n65, Alpha=Get thigh_rot_up_r(), Translation="-12,0,-13", Rotation="100,0,0", Scale="1,1,1") {
    rotationMode=BMM_Additive
    rotationSpace=BCS_BoneSpace
    scaleMode=BMM_Additive
    scaleSpace=BCS_BoneSpace
    translationMode=BMM_Additive
    translationSpace=BCS_BoneSpace
    alphaBoolBlend {blendOption=Linear}
    boneToModify {boneName=dyn_skirt_bk_r_MSTR}
    rotation {pitch=100}
    translation {x=-12, z=-13}
   }
   n67 = "Transform (Modify) Bone - Bone: dyn_skirt_bk_r_MSTR"(ComponentPose=@n66, Alpha=Get thigh_rot_out_r(), Translation="0,0,2", Rotation="0,0,30", Scale="1,1,1") {
    rotationMode=BMM_Additive
    rotationSpace=BCS_BoneSpace
    scaleMode=BMM_Additive
    scaleSpace=BCS_BoneSpace
    translationMode=BMM_Additive
    translationSpace=BCS_BoneSpace
    alphaBoolBlend {blendOption=Linear}
    boneToModify {boneName=dyn_skirt_bk_r_MSTR}
    rotation {roll=30}
    translation {z=2}
   }
   n68 = "Constraint - dyn_skirt_bk_r_MSTR (PP)"!(ComponentPose=@n67, Alpha=0, ConstraintWeights_0=0, ConstraintWeights_1:float=1) -> (Pose) {
    alphaBoolBlend {blendOption=Linear}
    boneToModify {boneName=dyn_skirt_bk_r_MSTR}
    constraintSetup [{targetBone {boneName=thigh_twist_01_r}}, {targetBone {boneName=pelvis}}]
    constraintWeights [1,1]
   }
   n69 = Get backSkirt_aim_alpha_r() -> (backSkirt_aim_alpha_r:double)
   n70 = Look At - Bone: dyn_skirt_bk_r_01(ComponentPose=@n68, Alpha=@n69) {
    alphaBoolBlend {blendOption=Linear}
    alphaScaleBiasClamp {bMapRange=true}
    boneToModify {boneName=dyn_skirt_bk_r_01}
    lookAtLocation=<8.0518 5.4639 -6.9372>
    lookAtTarget {boneReference {boneName=foot_r}}
    lookAt_Axis {axis=<1 0>}
   }
   n71 = Look At - Bone: dyn_skirt_bk_r_02(ComponentPose=@n70, Alpha=@n69) {
    alphaBoolBlend {blendOption=Linear}
    alphaScaleBiasClamp {bMapRange=true, outRange {max=0.5}}
    boneToModify {boneName=dyn_skirt_bk_r_02}
    lookAtLocation=<8.6532 6.7489 -10.1723>
    lookAtTarget {boneReference {boneName=foot_r}}
    lookAt_Axis {axis=<1 0>}
   }
   n72 = "Constraint - dyn_skirt_bk_l_MSTR (PP)"!(ComponentPose=@n71, Alpha=0, ConstraintWeights_0=0, ConstraintWeights_1:float=1) -> (Pose) {
    alphaBoolBlend {blendOption=Linear}
    boneToModify {boneName=dyn_skirt_bk_l_MSTR}
    constraintSetup [{targetBone {boneName=thigh_twist_01_l}}, {targetBone {boneName=pelvis}}]
    constraintWeights [1,1]
   }
   n73 = Get backSkirt_aim_alpha_l() -> (backSkirt_aim_alpha_l:double)
   n74 = Look At - Bone: dyn_skirt_bk_l_01(ComponentPose=@n72, Alpha=@n73) {
    alphaBoolBlend {blendOption=Linear}
    alphaScaleBiasClamp {bMapRange=true}
    boneToModify {boneName=dyn_skirt_bk_l_01}
    lookAtLocation=<4.3709 -39.9889 -11.3622>
    lookAtTarget {boneReference {boneName=foot_l}}
    lookAt_Axis {axis=<1 0>}
   }
   n75 = Look At - Bone: dyn_skirt_bk_l_02(ComponentPose=@n74, Alpha=@n73) {
    alphaBoolBlend {blendOption=Linear}
    alphaScaleBiasClamp {bMapRange=true, outRange {max=0.5}}
    boneToModify {boneName=dyn_skirt_bk_l_02}
    lookAtLocation=<-3.8987 -34.0151 -14.0405>
    lookAtTarget {boneReference {boneName=foot_l}}
    lookAt_Axis {axis=<1 0>}
   }
   n76 = "Transform (Modify) Bone - Bone: dyn_skirt_bk_r_01"(ComponentPose=@n75, Alpha=Get calf_rot_r() -> (calf_rot_r:double), Translation="0,0,0", Rotation="10,15,0", Scale="1,1,1") {
    rotationMode=BMM_Additive
    rotationSpace=BCS_BoneSpace
    scaleMode=BMM_Additive
    scaleSpace=BCS_BoneSpace
    translationMode=BMM_Additive
    translationSpace=BCS_BoneSpace
    alphaBoolBlend {blendOption=Linear}
    boneToModify {boneName=dyn_skirt_bk_r_01}
    rotation {pitch=10, yaw=15}
   }
   n77 = "Transform (Modify) Bone - Bone: dyn_skirt_bk_r_base"!(ComponentPose=@n76, Alpha=1, Translation="-2,0,5", Rotation="10,0,0") -> (Pose) {
    rotationMode=BMM_Replace
    rotationSpace=BCS_BoneSpace
    scaleMode=BMM_Additive
    scaleSpace=BCS_BoneSpace
    translationMode=BMM_Replace
    translationSpace=BCS_BoneSpace
    alphaBoolBlend {blendOption=Linear}
    alphaScaleBiasClamp {bClampResult=true, bInterpResult=true, interpSpeedDecreasing=5, interpSpeedIncreasing=7}
    boneToModify {boneName=dyn_skirt_bk_r_base}
    rotation {pitch=10}
    scale {x=0.7}
    translation {x=-2, z=5}
   }
   n78 = "Transform (Modify) Bone - Bone: dyn_skirt_bk_l_base"!(ComponentPose=@n77, Alpha=1, Translation="-2,0,4.5", Rotation="15,0,0") -> (Pose) {
    rotationMode=BMM_Replace
    rotationSpace=BCS_BoneSpace
    scaleMode=BMM_Additive
    scaleSpace=BCS_BoneSpace
    translationMode=BMM_Replace
    translationSpace=BCS_BoneSpace
    alphaBoolBlend {blendOption=Linear}
    alphaScaleBiasClamp {bClampResult=true, bInterpResult=true, interpSpeedDecreasing=5, interpSpeedIncreasing=7}
    boneToModify {boneName=dyn_skirt_bk_l_base}
    rotation {pitch=15}
    scale {x=0.9}
    translation {x=-2, z=4.5}
   }
   n79 = "Transform (Modify) Bone - Bone: dyn_skirt_bk_r_01"!(ComponentPose=@n78, Alpha=1, Translation="0,0,2", Rotation="10,0,0") -> (Pose) {
    rotationMode=BMM_Replace
    rotationSpace=BCS_BoneSpace
    scaleMode=BMM_Additive
    scaleSpace=BCS_BoneSpace
    translationMode=BMM_Replace
    translationSpace=BCS_BoneSpace
    alphaBoolBlend {blendOption=Linear}
    alphaScaleBiasClamp {bClampResult=true, bInterpResult=true, interpSpeedDecreasing=5, interpSpeedIncreasing=7}
    boneToModify {boneName=dyn_skirt_bk_r_01}
    rotation {pitch=10}
    scale {x=0.8}
    translation {z=2}
   }
   n80 = "Transform (Modify) Bone - Bone: dyn_skirt_ft_r_01"!(ComponentPose=@n79, Alpha=1, Translation="0,-2,0", Rotation="20,0,0") -> (Pose) {
    rotationMode=BMM_Replace
    rotationSpace=BCS_BoneSpace
    scaleMode=BMM_Additive
    scaleSpace=BCS_BoneSpace
    translationMode=BMM_Replace
    translationSpace=BCS_BoneSpace
    alphaBoolBlend {blendOption=Linear}
    alphaScaleBiasClamp {bClampResult=true, bInterpResult=true, interpSpeedDecreasing=5, interpSpeedIncreasing=7}
    boneToModify {boneName=dyn_skirt_ft_r_01}
    rotation {pitch=20}
    translation {y=-2}
   }
   Output Pose(Result=Blend Poses by bool(
     BlendPose_0=Component To Local(ComponentPose=@n57)
     BlendPose_1=Component To Local(ComponentPose=@n80)
     BlendTime_0=0.3
     BlendTime_1=0.3
     bActiveValue=Get bRigidBodyEnabled()
    ) {blendType=Linear}) {layerGroup=None}
   n81 = "Transform (Modify) Bone - Bone: JNT_Snake_01_ribbon_01_bind"(
    ComponentPose=Local To Component(LocalPose=Use cached pose 'PreArmCache'())
    Alpha=1
    Translation="0,0,0"
    Rotation="0,0,0"
    Scale="0.001,0.001,0.001"
   ) {scaleMode=BMM_Replace, alphaBoolBlend {blendOption=Linear}, boneToModify {boneName=JNT_Snake_01_ribbon_01_bind}, scale=<0.001 0.001 0.001>}
   n82 = Get Montage Timing() -> (Montage Timing:double)
   n83 = Control Rig!(
    Source=Layered blend per bone(
     BasePose=Component To Local(ComponentPose=@n81)
     BlendPoses_0="Sequence Evaluator (by time) 'Fangs_Emote_Full'"!(ExplicitTime:float=@n82) -> (Pose:Pose Link Structure) {
      is=title AnimGraphNode_SequenceEvaluator
      sequence=&/Fangs_Emote_Full
     }
     BlendWeights_0=Get Montage Is Active() -> (Montage Is Active:double)
    ) {layerSetup [{branchFilters [{boneName=attach_proxy}]}]}
    Alpha
    KneeRot_fr_l_114deg:Vector
    KneeTrans_fr_l_114deg:Vector="0.258,-3.591,0.009"
    KneeScale_fr_l_114deg:Vector="1,1,1"
    KneeScale_fr_l_130deg:Vector="1,1,1"
    KneeRot_fr_l_130deg:Vector="0,0,-51.2"
    KneeTrans_fr_l_130deg:Vector="0.258,-3.591,0.009"
   ) -> (Pose) {
    alphaBoolBlend {blendOption=Linear}
    controlRigAssetReference {blueprintRigClass=&/M_MED_Ensemble_Snake_Deform_CtrlRig_sculpters.M_MED_Ensemble_Snake_Deform_CtrlRig_sculpters_C}
   }
   n84 = "Transform (Modify) Bone - Bone: dyn_sleeve_bk_r"(
    ComponentPose=Local To Component(LocalPose=@n83)
    Alpha=Get arm_rot_up_r() -> (arm_rot_up_r:double)
    Translation="0,0,0"
    Rotation="-15,0,10"
    Scale="1,1,1"
   ) {
    rotationMode=BMM_Additive
    rotationSpace=BCS_BoneSpace
    translationMode=BMM_Additive
    translationSpace=BCS_BoneSpace
    alphaBoolBlend {blendOption=Linear}
    boneToModify {boneName=dyn_sleeve_bk_r}
    rotation {pitch=-15, roll=10}
   }
   n85 = Anim Dynamics - Bone: dyn_sleeve_up_r!(ComponentPose=@n84, Alpha=1, GravityScale:float=1, GravityOverride:Vector="0,0,-300", ExternalForce="0,0,0") -> (Pose) {
    angularDampingOverride=0.9999
    angularSpringConstant=500
    bGravityOverrideInSimSpace=true
    bOverrideAngularDamping=true
    simulationSpace=RootRelative
    alphaBoolBlend {blendOption=Linear}
    boundBone {boneName=dyn_sleeve_up_r}
    componentAppliedLinearAccClamp=<5000 5000 5000>
    componentLinearAccScale=<0.6 0.6 0.6>
    componentLinearVelScale=<0.3 0.3 0.3>
    gravityOverride {z=-980}
    physicsBodyDefinitions [{
      sphereCollisionRadius=0
      boundBone {boneName=dyn_sleeve_up_r}
      boxExtents=<12 15 10>
      constraintSetup {angularLimitsMax=<0 10 3>, angularLimitsMin=<0 -30 -3>, angularTarget=<0 0 0>}
      localJointOffset=<2 0 -1>
     }]
    relativeSpaceBone {boneName=head}
   }
   n86 = Anim Dynamics - Bone: dyn_sleeve_dn_r!(ComponentPose=@n85, Alpha=1, GravityScale:float=1, GravityOverride:Vector="0,0,-300", ExternalForce="0,0,0") -> (Pose) {
    angularDampingOverride=0.9999
    angularSpringConstant=500
    bGravityOverrideInSimSpace=true
    bOverrideAngularDamping=true
    simulationSpace=RootRelative
    alphaBoolBlend {blendOption=Linear}
    boundBone {boneName=dyn_sleeve_dn_r}
    componentAppliedLinearAccClamp=<5000 5000 5000>
    componentLinearAccScale=<0.6 0.6 0.6>
    componentLinearVelScale=<0.6 0.6 0.6>
    gravityOverride {z=-980}
    physicsBodyDefinitions [{
      sphereCollisionRadius=0
      boundBone {boneName=dyn_sleeve_dn_r}
      boxExtents=<12 15 10>
      constraintSetup {angularLimitsMax=<0 10 5>, angularLimitsMin=<0 -10 -5>, angularTarget=<0 0 0>}
      localJointOffset=<1.5 0 0.5>
     }]
    relativeSpaceBone {boneName=head}
   }
   n87 = Anim Dynamics - Bone: dyn_sleeve_bk_r!(ComponentPose=@n86, Alpha=1, GravityScale:float=1, GravityOverride:Vector="0,0,-300", ExternalForce="0,0,0") -> (Pose) {
    angularDampingOverride=0.9999
    angularSpringConstant=500
    bGravityOverrideInSimSpace=true
    bOverrideAngularDamping=true
    simulationSpace=RootRelative
    alphaBoolBlend {blendOption=Linear}
    boundBone {boneName=dyn_sleeve_bk_r}
    componentAppliedLinearAccClamp=<5000 5000 5000>
    componentLinearAccScale=<0.6 0.6 0.6>
    componentLinearVelScale=<0.6 0.6 0.6>
    gravityOverride {z=-980}
    physicsBodyDefinitions [{
      sphereCollisionRadius=0
      boundBone {boneName=dyn_sleeve_bk_r}
      boxExtents=<12 15 10>
      constraintSetup {angularLimitsMax=<0 5 10>, angularLimitsMin=<0 -5 -10>, angularTarget=<0 0 0>}
      localJointOffset=<1 1 0>
     }]
    relativeSpaceBone {boneName=head}
   }
   n88 = Anim Dynamics - Bone: dyn_sleeve_ft_r!(ComponentPose=@n87, Alpha=1, GravityScale:float=1, GravityOverride:Vector="0,0,-300", ExternalForce="0,0,0") -> (Pose) {
    angularDampingOverride=0.9999
    angularSpringConstant=500
    bGravityOverrideInSimSpace=true
    bOverrideAngularDamping=true
    simulationSpace=RootRelative
    alphaBoolBlend {blendOption=Linear}
    boundBone {boneName=dyn_sleeve_ft_r}
    componentAppliedLinearAccClamp=<5000 5000 5000>
    componentLinearAccScale=<0.6 0.6 0.6>
    componentLinearVelScale=<0.6 0.6 0.6>
    gravityOverride {z=-980}
    physicsBodyDefinitions [{
      sphereCollisionRadius=0
      boundBone {boneName=dyn_sleeve_ft_r}
      boxExtents=<12 15 10>
      constraintSetup {angularLimitsMax=<0 -10 20>, angularLimitsMin=<0 10 -20>, angularTarget=<0 0 0>}
      localJointOffset=<1 -1 0>
     }]
    relativeSpaceBone {boneName=head}
   }
   n89 = "Transform (Modify) Bone - Bone: dyn_sleeve_bk_l"(ComponentPose=@n88, Alpha=Get arm_rot_up_l() -> (arm_rot_up_l:double), Translation="0,0,0", Rotation="-15,0,-10", Scale="1,1,1") {
    rotationMode=BMM_Additive
    rotationSpace=BCS_BoneSpace
    translationMode=BMM_Additive
    translationSpace=BCS_BoneSpace
    alphaBoolBlend {blendOption=Linear}
    boneToModify {boneName=dyn_sleeve_bk_l}
    rotation {pitch=-15, roll=-10}
   }
   n90 = Anim Dynamics - Bone: dyn_sleeve_up_l!(ComponentPose=@n89, Alpha=1, GravityScale:float=1, GravityOverride:Vector="0,0,-300", ExternalForce="0,0,0") -> (Pose) {
    angularDampingOverride=0.9999
    angularSpringConstant=500
    bGravityOverrideInSimSpace=true
    bOverrideAngularDamping=true
    simulationSpace=RootRelative
    alphaBoolBlend {blendOption=Linear}
    boundBone {boneName=dyn_sleeve_up_l}
    componentAppliedLinearAccClamp=<5000 5000 5000>
    componentLinearAccScale=<0.6 0.6 0.6>
    componentLinearVelScale=<0.3 0.3 0.3>
    gravityOverride {z=-980}
    physicsBodyDefinitions [{
      sphereCollisionRadius=0
      boundBone {boneName=dyn_sleeve_up_l}
      boxExtents=<12 15 10>
      constraintSetup {angularLimitsMax=<0 10 3>, angularLimitsMin=<0 -40 -3>, angularTarget=<0 0 0>}
      localJointOffset=<2 0 -1>
     }]
    relativeSpaceBone {boneName=head}
   }
   n91 = Anim Dynamics - Bone: dyn_sleeve_dn_l!(ComponentPose=@n90, Alpha=1, GravityScale:float=1, GravityOverride:Vector="0,0,-300", ExternalForce="0,0,0") -> (Pose) {
    angularDampingOverride=0.9999
    angularSpringConstant=500
    bGravityOverrideInSimSpace=true
    bOverrideAngularDamping=true
    simulationSpace=RootRelative
    alphaBoolBlend {blendOption=Linear}
    boundBone {boneName=dyn_sleeve_dn_l}
    componentAppliedLinearAccClamp=<5000 5000 5000>
    componentLinearAccScale=<0.6 0.6 0.6>
    componentLinearVelScale=<0.6 0.6 0.6>
    gravityOverride {z=-980}
    physicsBodyDefinitions [{
      sphereCollisionRadius=0
      boundBone {boneName=dyn_sleeve_dn_l}
      boxExtents=<12 15 10>
      constraintSetup {angularLimitsMax=<0 10 5>, angularLimitsMin=<0 -10 -5>, angularTarget=<0 0 0>}
      localJointOffset=<1.5 0 0.5>
     }]
    relativeSpaceBone {boneName=head}
   }
   n92 = Anim Dynamics - Bone: dyn_sleeve_bk_l!(ComponentPose=@n91, Alpha=1, GravityScale:float=1, GravityOverride:Vector="0,0,-300", ExternalForce="0,0,0") -> (Pose) {
    angularDampingOverride=0.9999
    angularSpringConstant=500
    bGravityOverrideInSimSpace=true
    bOverrideAngularDamping=true
    simulationSpace=RootRelative
    alphaBoolBlend {blendOption=Linear}
    boundBone {boneName=dyn_sleeve_bk_l}
    componentAppliedLinearAccClamp=<5000 5000 5000>
    componentLinearAccScale=<0.6 0.6 0.6>
    componentLinearVelScale=<0.6 0.6 0.6>
    gravityOverride {z=-980}
    physicsBodyDefinitions [{
      sphereCollisionRadius=0
      boundBone {boneName=dyn_sleeve_bk_l}
      boxExtents=<12 15 10>
      constraintSetup {angularLimitsMax=<0 5 10>, angularLimitsMin=<0 -5 -10>, angularTarget=<0 0 0>}
      localJointOffset=<1 1 0>
     }]
    relativeSpaceBone {boneName=head}
   }
   n93 = Anim Dynamics - Bone: dyn_sleeve_ft_l!(ComponentPose=@n92, Alpha=1, GravityScale:float=1, GravityOverride:Vector="0,0,-300", ExternalForce="0,0,0") -> (Pose) {
    angularDampingOverride=0.9999
    angularSpringConstant=500
    bGravityOverrideInSimSpace=true
    bOverrideAngularDamping=true
    simulationSpace=RootRelative
    alphaBoolBlend {blendOption=Linear}
    boundBone {boneName=dyn_sleeve_ft_l}
    componentAppliedLinearAccClamp=<5000 5000 5000>
    componentLinearAccScale=<0.6 0.6 0.6>
    componentLinearVelScale=<0.6 0.6 0.6>
    gravityOverride {z=-980}
    physicsBodyDefinitions [{
      sphereCollisionRadius=0
      boundBone {boneName=dyn_sleeve_ft_l}
      boxExtents=<12 15 10>
      constraintSetup {angularLimitsMax=<0 -10 10>, angularLimitsMin=<0 10 -10>, angularTarget=<0 0 0>}
      localJointOffset=<1 -1 0>
     }]
    relativeSpaceBone {boneName=head}
   }
   n94 = "Constraint - dyn_knot_MSTR (PPPP)"!(
    ComponentPose=@n93
    Alpha=0
    ConstraintWeights_0=0.5
    ConstraintWeights_1:float=0.5
    ConstraintWeights_2:float=0
    ConstraintWeights_3:float=0
   ) -> (Pose) {
    alphaBoolBlend {blendOption=Linear}
    boneToModify {boneName=dyn_knot_MSTR}
    constraintSetup [{targetBone {boneName=spine_01}}, {targetBone {boneName=spine_02}}, {targetBone {boneName=pelvis}}, {targetBone {boneName=spine_03}}]
    constraintWeights [1,1,1,1]
   }
   n95 = "Transform (Modify) Bone - Bone: attach_backpack"(ComponentPose=@n94, Alpha=1, Translation="0,0,0", Rotation=Get socket_pin() -> (socket_pin:Rotator), Scale="1,1,1") {
    rotationMode=BMM_Additive
    rotationSpace=BCS_BoneSpace
    translationMode=BMM_Additive
    translationSpace=BCS_BoneSpace
    alphaBoolBlend {blendOption=Linear}
    boneToModify {boneName=attach_backpack}
   }
   n96 = "Transform (Modify) Bone - Bone: attach_backpack"(ComponentPose=@n95, Alpha=Get torso_rotate_bk() -> (torso_rotate_bk:double), Translation="0,0,0", Rotation="0,10,0", Scale="1,1,1") {
    rotationMode=BMM_Additive
    rotationSpace=BCS_BoneSpace
    translationMode=BMM_Additive
    translationSpace=BCS_BoneSpace
    alphaBoolBlend {blendOption=Linear}
    boneToModify {boneName=attach_backpack}
    rotation {yaw=10}
   }
   n97 = Get isWearBackbling() -> (isWearBackbling:bool)
   n98 = "Transform (Modify) Bone - Bone: dyn_knot_MSTR"(ComponentPose=@n96, bAlphaBoolEnabled=@n97, Translation="0,0,0", Rotation="0,0,0", Scale="1,1,1") {
    alphaInputType=Bool
    rotationMode=BMM_Additive
    rotationSpace=BCS_BoneSpace
    translationMode=BMM_Additive
    translationSpace=BCS_BoneSpace
    alphaBoolBlend {blendOption=Linear}
    boneToModify {boneName=dyn_knot_MSTR}
   }
   n99 = "Transform (Modify) Bone - Bone: dyn_bow_l_MSTR"!(ComponentPose=@n98, bAlphaBoolEnabled:bool=@n97, Translation="0,0,0", Rotation="30,-20,-5", Scale="1,1,1") -> (Pose) {
    alphaInputType=Bool
    rotationMode=BMM_Additive
    rotationSpace=BCS_BoneSpace
    scaleMode=BMM_Additive
    scaleSpace=BCS_BoneSpace
    translationMode=BMM_Additive
    translationSpace=BCS_BoneSpace
    alphaBoolBlend {blendOption=Linear}
    boneToModify {boneName=dyn_bow_l_MSTR}
    rotation=<rot 30 -20 -5>
   }
   n100 = "Transform (Modify) Bone - Bone: dyn_bow_r_MSTR"!(ComponentPose=@n99, bAlphaBoolEnabled:bool=@n97, Translation="0,0,0", Rotation="20,20,25", Scale="1,1,1") -> (Pose) {
    alphaInputType=Bool
    rotationMode=BMM_Additive
    rotationSpace=BCS_BoneSpace
    scaleMode=BMM_Additive
    scaleSpace=BCS_BoneSpace
    translationMode=BMM_Additive
    translationSpace=BCS_BoneSpace
    alphaBoolBlend {blendOption=Linear}
    boneToModify {boneName=dyn_bow_r_MSTR}
    rotation=<rot 20 20 25>
   }
   n101 = "Transform (Modify) Bone - Bone: dyn_ropes_r_MSTR"(ComponentPose=@n100, bAlphaBoolEnabled=@n97, Translation="-2.415,12.463,8.405", Rotation="18.059,8.608,-8.745", Scale="1,1,1") {
    alphaInputType=Bool
    rotationMode=BMM_Additive
    rotationSpace=BCS_BoneSpace
    scaleMode=BMM_Additive
    scaleSpace=BCS_BoneSpace
    translationMode=BMM_Additive
    translationSpace=BCS_BoneSpace
    alphaBoolBlend {blendOption=Linear}
    boneToModify {boneName=dyn_ropes_r_MSTR}
    rotation=<rot 18.059 8.608 -8.745>
    translation=<-2.415 12.463 8.405>
   }
   n102 = "Transform (Modify) Bone - Bone: dyn_ropes_l_MSTR"!(ComponentPose=@n101, bAlphaBoolEnabled:bool=@n97, Translation="-5,-5,8.000", Rotation="0,0,0", Scale="1,1,1") -> (Pose) {
    alphaInputType=Bool
    rotationMode=BMM_Additive
    rotationSpace=BCS_BoneSpace
    scaleMode=BMM_Additive
    scaleSpace=BCS_BoneSpace
    translationMode=BMM_Additive
    translationSpace=BCS_BoneSpace
    alphaBoolBlend {blendOption=Linear}
    boneToModify {boneName=dyn_ropes_l_MSTR}
    rotation=<rot 0 0 0>
    translation=<-5 -5 8>
   }
   n103 = "Transform (Modify) Bone - Bone: dyn_knot_MSTR"(ComponentPose=@n102, bAlphaBoolEnabled=Get bIsSkydiveFloating(), Translation="0,0,0", Rotation="-0.000,0.000,-25", Scale="1,1,1") {
    alphaInputType=Bool
    rotationMode=BMM_Additive
    rotationSpace=BCS_BoneSpace
    translationMode=BMM_Additive
    translationSpace=BCS_BoneSpace
    alphaBoolBlend {blendOption=Linear}
    boneToModify {boneName=dyn_knot_MSTR}
    rotation {roll=-25}
   }
   n104 = Get Head_Rot_R() -> (Head_Rot_R:double)
   n105 = "Transform (Modify) Bone - Bone: dyn_bow_r_MSTR"!(ComponentPose=@n103, Alpha:float=@n104, Translation="0.000,0,1.447", Rotation="25.000,0.000,0.000", Scale="1,1,1") -> (Pose) {
    rotationMode=BMM_Additive
    rotationSpace=BCS_BoneSpace
    scaleMode=BMM_Additive
    scaleSpace=BCS_BoneSpace
    translationMode=BMM_Additive
    translationSpace=BCS_BoneSpace
    alphaBoolBlend {blendOption=Linear}
    boneToModify {boneName=dyn_bow_r_MSTR}
    rotation=<rot 25 0.0001 0.0001>
    translation {z=1.447}
   }
   n106 = "Transform (Modify) Bone - Bone: dyn_ropes_r_MSTR"!(ComponentPose=@n105, Alpha:float=@n104, Translation="0.532,12.505,8.667", Rotation="0,-10.000,0.000", Scale="1,1,1") -> (Pose) {
    rotationMode=BMM_Additive
    rotationSpace=BCS_BoneSpace
    scaleMode=BMM_Additive
    scaleSpace=BCS_BoneSpace
    translationMode=BMM_Additive
    translationSpace=BCS_BoneSpace
    alphaBoolBlend {blendOption=Linear}
    boneToModify {boneName=dyn_ropes_r_MSTR}
    rotation {yaw=-10}
    translation=<0.532 12.505 8.667>
   }
   n107 = Get Idel Motage Alpha() -> (Idel Motage Alpha:bool)
   n108 = "Transform (Modify) Bone - Bone: dyn_knot_MSTR"(ComponentPose=@n106, bAlphaBoolEnabled=@n107, Translation="0,0,0", Rotation="0,-15,0", Scale="1,1,1") {
    alphaInputType=Bool
    rotationMode=BMM_Additive
    rotationSpace=BCS_BoneSpace
    scaleMode=BMM_Additive
    scaleSpace=BCS_BoneSpace
    translationMode=BMM_Additive
    translationSpace=BCS_BoneSpace
    alphaBoolBlend {blendOption=Linear}
    boneToModify {boneName=dyn_knot_MSTR}
    rotation {yaw=-15}
   }
   n109 = "Transform (Modify) Bone - Bone: dyn_bow_l_MSTR"(ComponentPose=@n108, bAlphaBoolEnabled=@n107, Translation="0,0,0", Rotation="30,0,0", Scale="1,1,1") {
    alphaInputType=Bool
    rotationMode=BMM_Additive
    rotationSpace=BCS_BoneSpace
    scaleMode=BMM_Additive
    scaleSpace=BCS_BoneSpace
    translationMode=BMM_Additive
    translationSpace=BCS_BoneSpace
    alphaBoolBlend {blendOption=Linear}
    boneToModify {boneName=dyn_bow_l_MSTR}
    rotation {pitch=30}
   }
   n110 = "Transform (Modify) Bone - Bone: dyn_ropes_l_MSTR"(ComponentPose=@n109, bAlphaBoolEnabled=@n107, Translation="0.000,-0.000,6.241", Rotation="7.789,-0.000,-0.000", Scale="1,1,1") {
    alphaInputType=Bool
    rotationMode=BMM_Additive
    rotationSpace=BCS_BoneSpace
    scaleMode=BMM_Additive
    scaleSpace=BCS_BoneSpace
    translationMode=BMM_Additive
    translationSpace=BCS_BoneSpace
    alphaBoolBlend {blendOption=Linear}
    boneToModify {boneName=dyn_ropes_l_MSTR}
    rotation=<rot 7.789 0 0>
    translation {z=6.241}
   }
   n111 = Get arm_rot_bk_r() -> (arm_rot_bk_r:double)
   n112 = "Transform (Modify) Bone - Bone: dyn_ropes_r_01"(ComponentPose=@n110, Alpha=@n111, Translation="0,0,0", Rotation="15,-15,0", Scale="1,1,1") {
    rotationMode=BMM_Additive
    rotationSpace=BCS_BoneSpace
    scaleMode=BMM_Additive
    scaleSpace=BCS_BoneSpace
    translationMode=BMM_Additive
    translationSpace=BCS_BoneSpace
    alphaBoolBlend {blendOption=Linear}
    boneToModify {boneName=dyn_ropes_r_01}
    rotation {pitch=15, yaw=-15}
   }
   n113 = "Transform (Modify) Bone - Bone: dyn_bow_r"(ComponentPose=@n112, Alpha=@n111, Translation="0,0,0", Rotation="40,0,0.000", Scale="1,1,1") {
    rotationMode=BMM_Additive
    rotationSpace=BCS_BoneSpace
    scaleMode=BMM_Additive
    scaleSpace=BCS_BoneSpace
    translationMode=BMM_Additive
    translationSpace=BCS_BoneSpace
    alphaBoolBlend {blendOption=Linear}
    boneToModify {boneName=dyn_bow_r}
    rotation {pitch=40}
   }
   n114 = Get arm_rot_bk_l() -> (arm_rot_bk_l:double)
   n115 = "Transform (Modify) Bone - Bone: dyn_bow_l"(ComponentPose=@n113, Alpha=@n114, Translation="0,0,0", Rotation="30,0,0.000", Scale="1,1,1") {
    rotationMode=BMM_Additive
    rotationSpace=BCS_BoneSpace
    scaleMode=BMM_Additive
    scaleSpace=BCS_BoneSpace
    translationMode=BMM_Additive
    translationSpace=BCS_BoneSpace
    alphaBoolBlend {blendOption=Linear}
    alphaScaleBiasClamp {bInterpResult=true, interpSpeedIncreasing=20}
    boneToModify {boneName=dyn_bow_l}
    rotation {pitch=30}
   }
   n116 = "Transform (Modify) Bone - Bone: dyn_ropes_l_01"(ComponentPose=@n115, Alpha=@n114, Translation="0,0,0", Rotation="0,5,0.000", Scale="1,1,1") {
    rotationMode=BMM_Additive
    rotationSpace=BCS_BoneSpace
    scaleMode=BMM_Additive
    scaleSpace=BCS_BoneSpace
    translationMode=BMM_Additive
    translationSpace=BCS_BoneSpace
    alphaBoolBlend {blendOption=Linear}
    alphaScaleBiasClamp {bInterpResult=true}
    boneToModify {boneName=dyn_ropes_l_01}
    rotation {yaw=5}
   }
   n117 = Get GravityOverrideRope()
   n118 = Anim Dynamics - Bone: dyn_ropes_l_MSTR!(ComponentPose=@n116, Alpha=1, GravityScale:float=1, GravityOverride:Vector=@n117, ExternalForce="0,0,0") -> (Pose) {
    angularDampingOverride=0.9999
    angularSpringConstant=500
    bGravityOverrideInSimSpace=true
    bOverrideAngularDamping=true
    simulationSpace=RootRelative
    alphaBoolBlend {blendOption=Linear}
    boundBone {boneName=dyn_ropes_l_MSTR}
    componentAppliedLinearAccClamp=<1000 1000 1000>
    componentLinearAccScale=<0.3 0.1 0.1>
    componentLinearVelScale=<0.5 0.5 0.5>
    gravityOverride {z=-980}
    physicsBodyDefinitions [{
      sphereCollisionRadius=0
      boundBone {boneName=dyn_ropes_l_MSTR}
      boxExtents=<12 15 10>
      constraintSetup {angularLimitsMax=<0 0 15>, angularLimitsMin=<0 0 -15>, angularTarget=<0 0 0>}
      localJointOffset=<-0.5 -0.3 0>
     }]
    relativeSpaceBone {boneName=head}
   }
   n119 = Anim Dynamics - Bone: dyn_ropes_r_MSTR!(ComponentPose=@n118, Alpha=1, GravityScale:float=1, GravityOverride:Vector=@n117, ExternalForce="0,0,0") -> (Pose) {
    angularDampingOverride=0.9999
    angularSpringConstant=500
    bGravityOverrideInSimSpace=true
    bOverrideAngularDamping=true
    simulationSpace=RootRelative
    alphaBoolBlend {blendOption=Linear}
    boundBone {boneName=dyn_ropes_r_MSTR}
    componentAppliedLinearAccClamp=<1000 1000 1000>
    componentLinearAccScale=<0.3 0.1 0.1>
    componentLinearVelScale=<0.5 0.5 0.5>
    gravityOverride {z=-980}
    physicsBodyDefinitions [{
      sphereCollisionRadius=0
      boundBone {boneName=dyn_ropes_r_MSTR}
      boxExtents=<12 15 10>
      constraintSetup {angularLimitsMax=<0 0 15>, angularLimitsMin=<0 0 -15>, angularTarget=<0 0 0>}
      localJointOffset=<-0.5 0.3 0>
     }]
    relativeSpaceBone {boneName=head}
   }
   n120 = Anim Dynamics - Bone: dyn_bow_l_MSTR!(ComponentPose=@n119, Alpha=1, GravityScale:float=1, GravityOverride:Vector=@n117, ExternalForce="0,0,0") -> (Pose) {
    angularDampingOverride=0.9999
    angularSpringConstant=500
    bGravityOverrideInSimSpace=true
    bOverrideAngularDamping=true
    simulationSpace=RootRelative
    alphaBoolBlend {blendOption=Linear}
    boundBone {boneName=dyn_bow_l_MSTR}
    componentAppliedLinearAccClamp=<1000 1000 1000>
    componentLinearAccScale=<0.3 0.1 0.1>
    componentLinearVelScale=<0.5 0.5 0.5>
    gravityOverride {z=-980}
    physicsBodyDefinitions [{
      sphereCollisionRadius=0
      boundBone {boneName=dyn_bow_l_MSTR}
      boxExtents=<12 15 10>
      constraintSetup {angularLimitsMax=<0 0 15>, angularLimitsMin=<0 0 -15>, angularTarget=<0 0 0>}
      localJointOffset=<0 -1 0>
     }]
    relativeSpaceBone {boneName=head}
   }
   n121 = Anim Dynamics - Bone: dyn_bow_r_MSTR!(ComponentPose=@n120, Alpha=1, GravityScale:float=1, GravityOverride:Vector=@n117, ExternalForce="0,0,0") -> (Pose) {
    angularDampingOverride=0.9999
    angularSpringConstant=500
    bGravityOverrideInSimSpace=true
    bOverrideAngularDamping=true
    simulationSpace=RootRelative
    alphaBoolBlend {blendOption=Linear}
    boundBone {boneName=dyn_bow_r_MSTR}
    componentAppliedLinearAccClamp=<1000 1000 1000>
    componentLinearAccScale=<0.3 0.1 0.1>
    componentLinearVelScale=<0.5 0.5 0.5>
    gravityOverride {z=-980}
    physicsBodyDefinitions [{
      sphereCollisionRadius=0
      boundBone {boneName=dyn_bow_r_MSTR}
      boxExtents=<12 15 10>
      constraintSetup {angularLimitsMax=<0 0 15>, angularLimitsMin=<0 0 -15>, angularTarget=<0 0 0>}
      localJointOffset=<0 1.5 0>
     }]
    relativeSpaceBone {boneName=head}
   }
   Save cached pose 'basePose'(Pose=Component To Local(ComponentPose=@n121)) {cacheName=basePose}
   Save cached pose 'preRigidPose'(Pose=Use cached pose 'basePose'()) {cacheName=preRigidPose}
   Save cached pose 'PreArmCache'(Pose=Copy Pose From Mesh() {bCopyCurves=true, bUseAttachedParent=true}) {cacheName=PreArmCache}
   Get AccScaleOverride() -> (AccScaleOverride:Vector)
   Get isWearBackbling()
  ]}
 EventGraph {nodes [
   Get Dynamic Material Instance(SkeletalMeshComponent=GetPartSkeletalMeshComponent(), ElementIndex=1)
   > Do Once()
   Event BlueprintInitializeAnimation()
   n1 = GetOwnerSkeletalMeshComponent()
   > Is Valid(InputObject=@n1)
   > Set Skeletal_Mesh(Skeletal_Mesh:Skeletal Mesh Component Object Reference=@n1) -> (Output_Get:Skeletal Mesh Component Object Reference)
   > n2 = Cast To FortPlayerPawn(Object=GetOwningActor())
   > Is Valid(InputObject=@n2)
   > n3 = Set Pawn(Pawn:Fort Player Pawn Object Reference=@n2) -> (Output_Get:Fort Player Pawn Object Reference)
   n4 = GetAnimInstance(self=Get Mesh(self=@n3))
   > Is Valid(InputObject=@n4)
   > Set AnimInst(AnimInst:Anim Instance Object Reference=@n4) -> (Output_Get:Anim Instance Object Reference)
   Event BlueprintUpdateAnimation()
   > n5 = Sequence!(execute) -> (then_0, then_1, then_2:Exec, then_3:Exec, then_4:Exec, then_5:Exec, then_6:Exec, then_7:Exec)
   n6 = GetOwnerSkeletalMeshComponent()
   n7 = BreakRotator(InRot=BreakTransform(InTransform=GetDeltaTransformFromRefPose(self=@n6, BoneName=thigh_r, BaseName=pelvis)).Rotation)
   > Set thigh_rot_up_r(thigh_rot_up_r:double=MapRangeClamped(Value=@n7.Yaw, InRangeA=-15, InRangeB=-110)) -> (Output_Get:double)
   > Set thigh_rot_out_r(thigh_rot_out_r:double=MapRangeClamped(Value=@n7.Pitch, InRangeA=0, InRangeB=35)) -> (Output_Get:double)
   n8 = GetOwnerSkeletalMeshComponent()
   n9 = BreakRotator(InRot=BreakTransform(InTransform=GetDeltaTransformFromRefPose(self=@n8, BoneName=thigh_l, BaseName=pelvis)).Rotation)
   > Set thigh_rot_up_l(thigh_rot_up_l:double=MapRangeClamped(Value=@n9.Yaw, InRangeA=-10, InRangeB=-110)) -> (Output_Get:double)
   > Set thigh_rot_in_l(thigh_rot_in_l:double=MapRangeClamped(Value=@n9.Pitch, InRangeA=0, InRangeB=-10)) -> (Output_Get:double)
   > Set thigh_rot_out_l(thigh_rot_out_l:double=MapRangeClamped(Value=@n9.Pitch, InRangeA=0, InRangeB=35)) -> (Output_Get:double)
   n10 = Get thigh_rot_up_l()
   > Set move_out_robe_l(move_out_robe_l:double=float * float(A=MapRangeClamped(Value=@n10, InRangeA=0, InRangeB=0.5)) {is=title K2Node_PromotableOperator}) -> (Output_Get:double)
   n11 = Get thigh_rot_up_r()
   > Set move_out_robe_r(move_out_robe_r:double=float * float(B=MapRangeClamped(Value=@n11, InRangeA=0, InRangeB=0.5))) -> (Output_Get:double)
   n12 = GetOwnerSkeletalMeshComponent()
   n13 = BreakRotator(InRot=BreakTransform(InTransform=GetDeltaTransformFromRefPose(self=@n12, BoneName=upperarm_r, BaseName=spine_05)).Rotation)
   Set arm_rot_bk_r(execute=@n5.then_1, arm_rot_bk_r:double=MapRangeClamped(Value=@n13.Yaw, InRangeA=0, InRangeB=-40)) -> (Output_Get:double)
   > Set arm_rot_up_r(arm_rot_up_r:double=MapRangeClamped(Value=@n13.Yaw, InRangeA=30, InRangeB=140)) -> (Output_Get:double)
   n14 = BreakRotator(InRot=BreakTransform(InTransform=GetDeltaTransformFromRefPose(self=@n12, BoneName=upperarm_l, BaseName=spine_04)).Rotation)
   > Set arm_rot_up_l(arm_rot_up_l:double=MapRangeClamped(Value=@n14.Yaw, InRangeA=20, InRangeB=100)) -> (Output_Get:double)
   > Set arm_rot_bk_l(arm_rot_bk_l:double=MapRangeClamped(Value=@n14.Yaw, InRangeA=0, InRangeB=-30)) -> (Output_Get:double)
   n15 = BreakTransform(InTransform=GetDeltaTransformFromRefPose(self=@n12, BoneName=clavicle_r, BaseName=spine_05))
   > Set clavicle_rot_up_r(clavicle_rot_up_r:double=MapRangeClamped(Value=BreakRotator(InRot=@n15.Rotation).Pitch, InRangeA=10, InRangeB=40)) -> (Output_Get:double)
   n16 = Get bIsCrouching()
   Set sprinting_alpha(
    execute=@n5.then_2
    sprinting_alpha:bool=AND Boolean!(A=Get bIsSprinting(), B=Get bIsOnGround() -> (bIsOnGround:bool), C:bool=NOT Boolean(A=@n16)) -> (ReturnValue)
   ) -> (Output_Get:bool)
   n17 = GetOwnerSkeletalMeshComponent()
   n18 = BreakRotator(InRot=BreakTransform(InTransform=GetDeltaTransformFromRefPose(self=@n17, BoneName=spine_05, BaseName=spine_02)).Rotation)
   > Set socket_pin(socket_pin:Rotator=MakeRotator(Roll=MapRangeClamped(Value=@n18, InRangeA=-50, InRangeB=50, OutRangeA=25, OutRangeB=-25))) -> (Output_Get:Rotator)
   > Set torso_rotate_bk(torso_rotate_bk:double=MapRangeClamped(Value=@n18.Yaw, InRangeA=0, InRangeB=-10)) -> (Output_Get:double)
   n19 = GetOwnerSkeletalMeshComponent()
   n20 = BreakRotator(InRot=BreakTransform(InTransform=GetDeltaTransformFromRefPose(self=@n19, BoneName=head, BaseName=spine_05)).Rotation)
   > Set Head_Rot_Bk(Head_Rot_Bk:double=MapRangeClamped(Value=@n20.Yaw, InRangeA=-10, InRangeB=-55)) -> (Output_Get:double)
   > Set Head_Rot_R(Head_Rot_R:double=MapRangeClamped(Value=@n20.Pitch, InRangeA=0, InRangeB=45)) -> (Output_Get:double)
   n21 = Gravity Override State Function DEPRECATED(
    execute=@n5.then_3
    CharacterPartAnimInstance=Self-Reference()
    GameTimeInSeconds=GetGameTimeInSeconds()
    GravityOverrideParams=Get Gravity Override Params()
   )
   > Set GravityOverride(GravityOverride=@n21)
   > n22 = Branch(Condition=OR Boolean(A=Get bIsSkyDiving(), B=Get bIsCrouching()))
   > Set AccScaleOverride(AccScaleOverride:Vector="0.1,0.1,0.05") -> (Output_Get:Vector)
   Set AccScaleOverride(execute=@n22.else, AccScaleOverride="0.15,0.15,0.1")
   n23 = GetOwnerSkeletalMeshComponent()
   n24 = BreakTransform(InTransform=GetDeltaTransformFromRefPose(self=@n23, BoneName=thigh_l, BaseName=pelvis))
   Set backSkirt_aim_alpha_l(execute=@n5.then_4, backSkirt_aim_alpha_l:double=MapRangeClamped(Value=BreakRotator(InRot=@n24.Rotation).Yaw, InRangeA=-100, InRangeB=-25)) -> (Output_Get:double)
   n25 = BreakTransform(InTransform=GetDeltaTransformFromRefPose(self=@n23, BoneName=thigh_r, BaseName=pelvis))
   > Set backSkirt_aim_alpha_r(backSkirt_aim_alpha_r:double=MapRangeClamped(Value=BreakRotator(InRot=@n25.Rotation).Yaw, InRangeA=-100, InRangeB=-25)) -> (Output_Get:double)
   n26 = BreakTransform(InTransform=GetDeltaTransformFromRefPose(self=@n23, BoneName=calf_r, BaseName=thigh_r))
   > Set calf_rot_r(calf_rot_r:double=MapRangeClamped(Value=BreakRotator(InRot=@n26.Rotation).Yaw, InRangeA=0, InRangeB=110)) -> (Output_Get:double)
   n27 = Get AnimInst() -> (AnimInst:Anim Instance Object Reference)
   Set Montage Is Active(execute=@n5.then_5, Montage Is Active:double=GetCurveValue(self=@n27, CurveName=emote_BuiltIn)) -> (Output_Get:double)
   > Set Montage Timing(Montage Timing:double=Montage_GetPosition!(
     self:Anim Instance Object Reference=@n27
     Montage:Anim Montage Object Reference=/BRCosmetics/Animation/Game/MainPlayer/Emotes/Fangs/Emote_Fangs_CMM_M
    ) -> (ReturnValue:float)) -> (Output_Get:double)
   > n28 = Branch(Condition=Get bIsInFrontEnd())
   > n29 = Branch(Condition=Get isWearBackbling())
   Set Idel Motage Alpha(execute=[@n28.else, @n29.then], Idel Motage Alpha:bool=false) -> (Output_Get:bool)
   Set Idel Motage Alpha(
    execute=@n29.else
    Idel Motage Alpha=Montage_IsActive!(
     self:Anim Instance Object Reference=@n27
     Montage:Anim Montage Object Reference=/BRCosmetics/Animation/Game/MainPlayer/Menu/BR/CustomSkins/_s21/EnsembleSnakeM/EnsembleSnake_Male_Idle_Montage
    ) -> (ReturnValue:bool)
   )
   n30 = Set isWearBackbling(execute=@n5.then_6, isWearBackbling:bool=CheckForBackbling().bIsBackblingACape) -> (Output_Get:bool)
   > n31 = Get Dynamic Material Instance(SkeletalMeshComponent=GetPartSkeletalMeshComponent(), ElementIndex=1)
   > n32 = Branch(Condition=@n30)
   > SetScalarParameterValue(self=@n31, ParameterName=Hide Element 3_5, Value=1)
   SetScalarParameterValue(execute=@n32.else, self=@n31, ParameterName=Hide Element 3_5, Value=0)
   n33 = Branch(execute=@n5.then_7, Condition=Get bIsSkydiveDiveMode())
   > Set GravityOverrideRope(GravityOverrideRope:Vector="0, 0, 0") -> (Output_Get:Vector)
   n34 = Branch(execute=@n33.else, Condition=Get bIsSkydiveFloating())
   > Set GravityOverrideRope(GravityOverrideRope="0,-100,100")
   Set GravityOverrideRope(execute=@n34.else, GravityOverrideRope="0,0,-300")
   Gravity Override State Function DEPRECATED(
    CharacterPartAnimInstance=Self-Reference()
    GameTimeInSeconds=GetGameTimeInSeconds()
    GravityOverrideParams=Get Gravity Override Params Rope() -> (Gravity Override Params Rope:Gravity Override Params Struct Structure)
   )
   SelectFloat(A=1, B=0, bPickA=Montage_IsActive(Montage=&Emote_Fangs_CMM_M))
   Get bIsPlayingEmote()
  ]}
]
variables [
 AccScaleOverride {type=Vector}
 AnimInst {type=AnimInstance, default=null}
 arm_rot_bk_l {type=number}
 arm_rot_bk_r {type=number}
 arm_rot_up_l {type=number}
 arm_rot_up_r {type=number}
 backSkirt_aim_alpha_l {type=number}
 backSkirt_aim_alpha_r {type=number}
 calf_rot_r {type=number}
 clavicle_rot_up_r {type=number}
 Gravity Override Params {
  type=GravityOverrideParamsStruct
  default {
   skydive_FloatingDown {gravityOverride=<0 -700 -300>, windAmplitude=<200 -500 50>, windOffset=<0 -600 50>}
   skydive_GoingUp {gravityOverride=<0 -300 1000>, windAmplitude=<200 -500 200>}
  }
 }
 Gravity Override Params Rope {
  type=GravityOverrideParamsStruct
  default {
   dBNO {gravityOverride=<780 0 0>}
   dBNO_Carried {gravityOverride=<780 0 0>}
   emoteOrMelee {gravityOverride=<0 0 780>}
   front_End {gravityOverride=<0 0 780>}
   onGround_Falling {gravityOverride=<780 0 0>}
   onGround_Moving {gravityOverride=<0 0 780>}
   onGround_Standing {gravityOverride=<0 0 780>}
   skydive_DivingDown {gravityOverride=<1500 0 0>}
   skydive_FloatingDown {jointName=pelvis, gravityOverride=<780 0 0>, windAmplitude=<0 0 0>, windFrequency=<0 0 0>, windOffset=<0 0 0>}
   skydive_GoingUp {gravityOverride=<0 0 1000>, windAmplitude=<0 0 0>, windFrequency=<0 0 0>}
   skydive_Parachuting {gravityOverride=<0 0 780>, windAmplitude=<0 0 0>, windFrequency=<0 0 0>}
   swimming {gravityOverride=<780 0 0>}
   ziplining {gravityOverride=<780 0 0>}
  }
 }
 GravityOverride {type=Vector}
 GravityOverrideRope {type=Vector, default=<0 0 -300>}
 Head_Rot_Bk {type=number}
 Head_Rot_R {type=number}
 Idel Motage Alpha {type=boolean}
 isWearBackbling {type=boolean}
 Montage Is Active {type=number}
 Montage Timing {type=number}
 move_out_robe_l {type=number}
 move_out_robe_r {type=number}
 Pawn {type=FortPlayerPawn}
 Skeletal_Mesh {type=SkeletalMeshComponent}
 socket_pin {type=Rotator}
 sprinting_alpha {type=boolean}
 thigh_rot_in_l {type=number}
 thigh_rot_out_l {type=number}
 thigh_rot_out_r {type=number}
 thigh_rot_up_l {type=number}
 thigh_rot_up_r {type=number}
 torso_rotate_bk {type=number}
]
