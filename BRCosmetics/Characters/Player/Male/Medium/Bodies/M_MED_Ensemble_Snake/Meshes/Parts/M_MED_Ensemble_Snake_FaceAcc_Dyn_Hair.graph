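# Physics asset for the dynamic hair on the Snake ensemble's face accessory mesh, for the medium male player character.
# Capsules on head, neck, spine, clavicles and upper arms form the body that the hair can collide with.
# Back, left, right-side, front and tip chains hang from the head, mostly through free MSTR joints, with angle-limited segments below.
# Free constraints also tie the left and back-right chain ends to the center back chain.
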PhysicsAsset /BRCosmetics/Characters/Player/Male/Medium/Bodies/M_MED_Ensemble_Snake/Meshes/Parts/M_MED_Ensemble_Snake_FaceAcc_Dyn_Hair
bodies {
 clavicle_l [{
   shapeType=Capsule
   length=4.5
   radius=8.25
   shapeName=clavicle_l_capsule
   center=<8.5775 -1.5592 -0.745>
   rotation=<rot 86.2624 -0.0144 -5.0145>
  }]
 clavicle_r [{
   shapeType=Capsule
   length=3.375
   radius=7.875
   shapeName=clavicle_l_capsule
   center=<-10.1806 1.7695 0.6913>
   rotation=<rot 86.2562 6.6637 -173.3696>
  }]
 dyn_hair_bk_00 [{shapeType=Capsule, length=0.505, radius=0.505, shapeName=dyn_hair_bk_MSTR_capsule, rotation=<rot 0 -90 0>}]
 dyn_hair_bk_01 [{shapeType=Capsule, length=7.25, radius=5.375, shapeName=dyn_hair_bk_01_capsule, center=<13 0 0>, rotation=<rot 85.0005 -180 -180>}]
 dyn_hair_bk_02 [{shapeType=Capsule, length=7.25, radius=6.375, shapeName=dyn_hair_bk_02_capsule, center=<12 0 3>, rotation=<rot 84.9999 0 0>}]
 dyn_hair_bk_AIM [{shapeType=Capsule, length=0.505, radius=0.505, shapeName=dyn_hair_bk_MSTR_capsule, rotation=<rot 0 -90 0>}]
 dyn_hair_bk_MSTR [{shapeType=Capsule, length=0.505, radius=0.505, shapeName=dyn_hair_bk_MSTR_capsule, rotation=<rot 0 -90 0>}]
 dyn_hair_bk_r_00 [{shapeType=Capsule, length=0.505, radius=0.505, shapeName=dyn_hair_bk_MSTR_capsule, rotation=<rot 0 -90 0>}]
 dyn_hair_bk_r_01 [{shapeType=Capsule, length=7.125, radius=5.125, shapeName=dyn_hair_r_01_capsule, center=<15 0 2>, rotation=<rot 90 0.8371 0.8371>}]
 dyn_hair_bk_r_02 [{shapeType=Capsule, length=4.5, radius=4.875, shapeName=dyn_hair_r_02_capsule, center=<8 0 3>, rotation=<rot 90 0 0>}]
 dyn_hair_bk_r_AIM [{shapeType=Capsule, length=0.505, radius=0.505, shapeName=dyn_hair_r_MSTR_capsule, rotation=<rot 0 -90 0>}]
 dyn_hair_bk_r_MSTR [{shapeType=Capsule, length=0.505, radius=0.505, shapeName=dyn_hair_r_MSTR_capsule, rotation=<rot 0 -90 0>}]
 dyn_hair_ft_l_01 [{shapeType=Capsule, length=0.505, radius=0.505, shapeName=dyn_hair_l_MSTR_capsule, rotation=<rot 0 -90 0>}]
 dyn_hair_ft_l_02 [{shapeType=Capsule, length=4.375, radius=3.125, shapeName=dyn_hair_l_01_capsule, center=<5 0 1>, rotation=<rot 90 156.0679 151.0686>}]
 dyn_hair_ft_l_03 [{shapeType=Capsule, length=6.875, radius=2.25, shapeName=dyn_hair_l_01_capsule, center=<8 0 0>, rotation=<rot 90 172.875 162.8747>}]
 dyn_hair_ft_l_MSTR [{shapeType=Capsule, length=0.505, radius=0.505, shapeName=dyn_hair_l_MSTR_capsule, rotation=<rot 0 -90 0>}]
 dyn_hair_ft_r_01 [{shapeType=Capsule, length=0.505, radius=0.505, shapeName=dyn_hair_l_MSTR_capsule, rotation=<rot 0 -90 0>}]
 dyn_hair_ft_r_02 [{shapeType=Capsule, length=4.375, radius=3.125, shapeName=dyn_hair_l_01_capsule, center=<3 0 1>, rotation=<rot 0 0 90>}]
 dyn_hair_ft_r_03 [{shapeType=Capsule, length=6.875, radius=2.625, shapeName=dyn_hair_l_01_capsule, center=<8 0 0>, rotation=<rot 84.9997 150.0037 -9.9954>}]
 dyn_hair_ft_r_MSTR [{shapeType=Capsule, length=0.505, radius=0.505, shapeName=dyn_hair_l_MSTR_capsule, rotation=<rot 0 -90 0>}]
 dyn_hair_l_00 [{shapeType=Capsule, length=0.505, radius=0.505, shapeName=dyn_hair_l_MSTR_capsule, rotation=<rot 0 -90 0>}]
 dyn_hair_l_01 [{shapeType=Capsule, length=10.25, radius=4.75, shapeName=dyn_hair_r_01_capsule, center=<16.5055 -2 5.2654>, rotation=<rot 80.0008 180 180>}]
 dyn_hair_l_02 [{shapeType=Capsule, length=2.875, radius=3.75, shapeName=dyn_hair_l_02_capsule, center=<8.5409 0 2.7399>, rotation=<rot 90 0.9734 0.9734>}]
 dyn_hair_l_AIM [{shapeType=Capsule, length=0.505, radius=0.505, shapeName=dyn_hair_l_MSTR_capsule, rotation=<rot 0 -90 0>}]
 dyn_hair_l_MSTR [{shapeType=Capsule, length=0.505, radius=0.505, shapeName=dyn_hair_l_MSTR_capsule, rotation=<rot 0 -90 0>}]
 dyn_hair_side_r_01 [{shapeType=Capsule, length=0.505, radius=0.505, shapeName=dyn_hair_r_MSTR_capsule, rotation=<rot 0 -90 0>}]
 dyn_hair_side_r_02 [{shapeType=Capsule, length=7.875, radius=4.625, shapeName=dyn_hair_r_02_capsule, center=<13 3 0>, rotation=<rot 90 71.0877 61.0886>}]
 dyn_hair_side_r_AIM [{shapeType=Capsule, length=0.505, radius=0.505, shapeName=dyn_hair_r_MSTR_capsule, rotation=<rot 0 -90 0>}]
 dyn_hair_side_r_MSTR [{shapeType=Capsule, length=0.505, radius=0.505, shapeName=dyn_hair_r_MSTR_capsule, rotation=<rot 0 -90 0>}]
 dyn_hair_tips_01 [{shapeType=Capsule, length=3.5, radius=2.125, shapeName=dyn_hair_l_01_capsule, center=<3.5 0 2>, rotation=<rot 90 0 0>}]
 dyn_hair_tips_02 [{shapeType=Capsule, length=3.5, radius=1.875, shapeName=dyn_hair_l_01_capsule, center=<3.5 0 2>, rotation=<rot 90 0 0>}]
 dyn_hair_tips_03 [{shapeType=Capsule, length=2, radius=2.25, shapeName=dyn_hair_l_01_capsule, center=<7 0 1>, rotation=<rot 85.0001 0 -0.0001>}]
 head [
  {shapeType=Capsule, length=6.0715, radius=6.6867, shapeName=head_capsule, center=<6.2684 -0.2133 0>, rotation=<rot 90 179.9899 179.9903>}
  {shapeType=Capsule, length=5.5715, radius=5.4367, shapeName=head_capsule, center=<3.6847 3.5395 0>, rotation=<rot 90 5.7106 0.7126>}
  {shapeType=Capsule, length=5.4465, radius=2.4367, shapeName=head_capsule, center=<2.9584 7.8104 0>, rotation=<rot 90 0 0.0021>}
 ]
 neck_01 [{shapeType=Capsule, length=5.4465, radius=5.6867, shapeName=head_capsule, center=<3.4473 -0.7185 0>, rotation=<rot 90 180 -174.9995>}]
 spine_02 [{shapeType=Capsule, length=3.625, radius=13.75, shapeName=spine_05_capsule, center=<-1.8484 2.7149 0>}]
 spine_03 [{shapeType=Capsule, length=4.875, radius=12.125, shapeName=spine_05_capsule, center=<6.0816 0.0577 0>}]
 spine_04 [{shapeType=Capsule, length=6.125, radius=11.125, shapeName=spine_05_capsule, center=<8.5157 -0.7666 0>}]
 spine_05 [{shapeType=Capsule, length=11.375, radius=8.625, shapeName=spine_05_capsule, center=<4.0157 -1.4341 0>}]
 upperarm_l [{
   shapeType=Capsule
   length=23.625
   radius=6.25
   shapeName=clavicle_l_capsule
   center=<14.2843 -2.6547 -2.3561>
   rotation=<rot 90 180 179.9998>
  }]
 upperarm_r [{
   shapeType=Capsule
   length=20.625
   radius=7
   shapeName=clavicle_l_capsule
   center=<-13.4743 3.0682 2.3561>
   rotation=<rot 90 -144.7357 35.2641>
  }]
}
constraints [
 {
  bone1Name=dyn_hair_l_MSTR
  bone2Name=head
  swing1LimitDegrees=45
  swing1Motion=Free
  swing2LimitDegrees=45
  swing2Motion=Free
  twistLimitDegrees=45
  twistMotion=Free
 }
 {bone1Name=dyn_hair_l_01, bone2Name=dyn_hair_l_AIM, swing1LimitDegrees=30, swing2LimitDegrees=60, twistLimitDegrees=5}
 {bone1Name=dyn_hair_l_02, bone2Name=dyn_hair_l_01, swing1LimitDegrees=30, swing2LimitDegrees=60, twistLimitDegrees=5}
 {
  bone1Name=dyn_hair_bk_MSTR
  bone2Name=head
  swing1LimitDegrees=45
  swing1Motion=Free
  swing2LimitDegrees=45
  swing2Motion=Free
  twistLimitDegrees=45
  twistMotion=Free
 }
 {bone1Name=dyn_hair_bk_01, bone2Name=dyn_hair_bk_AIM, swing1LimitDegrees=30, swing2LimitDegrees=60, twistLimitDegrees=5}
 {bone1Name=dyn_hair_bk_02, bone2Name=dyn_hair_bk_01, swing1LimitDegrees=30, swing2LimitDegrees=60, twistLimitDegrees=5}
 {
  bone1Name=head
  bone2Name=neck_01
  swing1LimitDegrees=45
  swing1Motion=Free
  swing2LimitDegrees=45
  swing2Motion=Free
  twistLimitDegrees=45
  twistMotion=Free
 }
 {
  bone1Name=spine_05
  bone2Name=spine_04
  swing1LimitDegrees=45
  swing1Motion=Free
  swing2LimitDegrees=45
  swing2Motion=Free
  twistLimitDegrees=45
  twistMotion=Free
 }
 {
  bone1Name=neck_01
  bone2Name=spine_05
  swing1LimitDegrees=45
  swing1Motion=Free
  swing2LimitDegrees=45
  swing2Motion=Free
  twistLimitDegrees=45
  twistMotion=Free
 }
 {
  bone1Name=upperarm_r
  bone2Name=clavicle_r
  swing1LimitDegrees=45
  swing1Motion=Free
  swing2LimitDegrees=45
  swing2Motion=Free
  twistLimitDegrees=45
  twistMotion=Free
 }
 {
  bone1Name=spine_04
  bone2Name=spine_03
  swing1LimitDegrees=45
  swing1Motion=Free
  swing2LimitDegrees=45
  swing2Motion=Free
  twistLimitDegrees=45
  twistMotion=Free
 }
 {bone1Name=dyn_hair_bk_00, bone2Name=dyn_hair_bk_MSTR, swing1LimitDegrees=5, swing2LimitDegrees=10, twistLimitDegrees=2}
 {bone1Name=dyn_hair_l_00, bone2Name=dyn_hair_l_MSTR, swing1LimitDegrees=5, swing2LimitDegrees=10, twistLimitDegrees=2}
 {
  bone1Name=dyn_hair_ft_r_MSTR
  bone2Name=head
  swing1LimitDegrees=45
  swing1Motion=Free
  swing2LimitDegrees=45
  swing2Motion=Free
  twistLimitDegrees=45
  twistMotion=Free
 }
 {bone1Name=dyn_hair_ft_r_01, bone2Name=dyn_hair_ft_r_MSTR, swing1LimitDegrees=10, swing2LimitDegrees=5, twistLimitDegrees=2}
 {bone1Name=dyn_hair_ft_r_02, bone2Name=dyn_hair_ft_r_01, swing1LimitDegrees=10, swing2LimitDegrees=5, twistLimitDegrees=2}
 {bone1Name=dyn_hair_ft_r_03, bone2Name=dyn_hair_ft_r_02, swing1LimitDegrees=35, swing2LimitDegrees=15, twistLimitDegrees=2}
 {
  bone1Name=dyn_hair_ft_l_MSTR
  bone2Name=head
  swing1LimitDegrees=45
  swing1Motion=Free
  swing2LimitDegrees=45
  swing2Motion=Free
  twistLimitDegrees=45
  twistMotion=Free
 }
 {bone1Name=dyn_hair_ft_l_01, bone2Name=dyn_hair_ft_l_MSTR, swing1LimitDegrees=5, swing2LimitDegrees=5, twistLimitDegrees=2}
 {bone1Name=dyn_hair_ft_l_02, bone2Name=dyn_hair_ft_l_01, swing1LimitDegrees=10, swing2LimitDegrees=5, twistLimitDegrees=2}
 {bone1Name=dyn_hair_ft_l_03, bone2Name=dyn_hair_ft_l_02, swing1LimitDegrees=35, swing2LimitDegrees=15, twistLimitDegrees=2}
 {
  bone1Name=clavicle_l
  bone2Name=spine_05
  swing1LimitDegrees=45
  swing1Motion=Free
  swing2LimitDegrees=45
  swing2Motion=Free
  twistLimitDegrees=45
  twistMotion=Free
 }
 {
  bone1Name=spine_03
  bone2Name=spine_02
  swing1LimitDegrees=45
  swing1Motion=Free
  swing2LimitDegrees=45
  swing2Motion=Free
  twistLimitDegrees=45
  twistMotion=Free
 }
 {bone1Name=dyn_hair_tips_01, bone2Name=head, swing1LimitDegrees=20, swing2LimitDegrees=35, twistLimitDegrees=5}
 {bone1Name=dyn_hair_tips_02, bone2Name=dyn_hair_tips_01, swing1LimitDegrees=20, swing2LimitDegrees=35, twistLimitDegrees=5}
 {bone1Name=dyn_hair_tips_03, bone2Name=head, swing1LimitDegrees=35, swing2LimitDegrees=15, twistLimitDegrees=5}
 {
  bone1Name=clavicle_r
  bone2Name=spine_05
  swing1LimitDegrees=45
  swing1Motion=Free
  swing2LimitDegrees=45
  swing2Motion=Free
  twistLimitDegrees=45
  twistMotion=Free
 }
 {
  bone1Name=upperarm_l
  bone2Name=clavicle_l
  swing1LimitDegrees=45
  swing1Motion=Free
  swing2LimitDegrees=45
  swing2Motion=Free
  twistLimitDegrees=45
  twistMotion=Free
 }
 {
  bone1Name=dyn_hair_l_AIM
  bone2Name=dyn_hair_l_00
  swing1LimitDegrees=45
  swing1Motion=Free
  swing2LimitDegrees=45
  swing2Motion=Free
  twistLimitDegrees=45
  twistMotion=Free
 }
 {
  bone1Name=dyn_hair_bk_AIM
  bone2Name=dyn_hair_bk_00
  swing1LimitDegrees=45
  swing1Motion=Free
  swing2LimitDegrees=45
  swing2Motion=Free
  twistLimitDegrees=45
  twistMotion=Free
 }
 {
  bone1Name=dyn_hair_bk_r_MSTR
  bone2Name=head
  swing1LimitDegrees=45
  swing1Motion=Free
  swing2LimitDegrees=45
  swing2Motion=Free
  twistLimitDegrees=45
  twistMotion=Free
 }
 {
  bone1Name=dyn_hair_bk_r_00
  bone2Name=dyn_hair_bk_r_MSTR
  swing1LimitDegrees=45
  swing1Motion=Free
  swing2LimitDegrees=45
  swing2Motion=Free
  twistLimitDegrees=45
  twistMotion=Free
 }
 {
  bone1Name=dyn_hair_bk_r_AIM
  bone2Name=dyn_hair_bk_r_00
  swing1LimitDegrees=45
  swing1Motion=Free
  swing2LimitDegrees=45
  swing2Motion=Free
  twistLimitDegrees=45
  twistMotion=Free
 }
 {bone1Name=dyn_hair_bk_r_01, bone2Name=dyn_hair_bk_r_AIM, swing1LimitDegrees=30, swing2LimitDegrees=60, twistLimitDegrees=5}
 {bone1Name=dyn_hair_bk_r_02, bone2Name=dyn_hair_bk_r_01, swing1LimitDegrees=30, swing2LimitDegrees=60, twistLimitDegrees=5}
 {bone1Name=dyn_hair_side_r_01, bone2Name=dyn_hair_side_r_MSTR, swing1LimitDegrees=5, swing2LimitDegrees=2, twistLimitDegrees=2}
 {
  bone1Name=dyn_hair_side_r_MSTR
  bone2Name=head
  swing1LimitDegrees=45
  swing1Motion=Free
  swing2LimitDegrees=45
  swing2Motion=Free
  twistLimitDegrees=45
  twistMotion=Free
 }
 {bone1Name=dyn_hair_side_r_02, bone2Name=dyn_hair_side_r_AIM, swing1LimitDegrees=25, swing2LimitDegrees=10, twistLimitDegrees=2}
 {
  bone1Name=dyn_hair_side_r_AIM
  bone2Name=dyn_hair_side_r_01
  swing1LimitDegrees=45
  swing1Motion=Free
  swing2LimitDegrees=45
  swing2Motion=Free
  twistLimitDegrees=45
  twistMotion=Free
 }
 {
  bone1Name=dyn_hair_bk_r_02
  bone2Name=dyn_hair_bk_02
  swing1LimitDegrees=45
  swing1Motion=Free
  swing2LimitDegrees=45
  swing2Motion=Free
  twistLimitDegrees=45
  twistMotion=Free
 }
 {
  bone1Name=dyn_hair_l_02
  bone2Name=dyn_hair_bk_02
  swing1LimitDegrees=45
  swing1Motion=Free
  swing2LimitDegrees=45
  swing2Motion=Free
  twistLimitDegrees=45
  twistMotion=Free
 }
]
