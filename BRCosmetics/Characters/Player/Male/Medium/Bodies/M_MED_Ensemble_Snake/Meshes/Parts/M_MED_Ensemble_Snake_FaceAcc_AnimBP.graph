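# Post-process style AnimBlueprint for the Snake ensemble face accessory, driving earring and hair secondary motion on a copied body pose.
# Anim Dynamics moves earrings and hair tips, then bRigidBodyEnabled picks between a RigidBody hair simulation and a constraint-based fallback.
# Head and right clavicle rotation relative to spine_05 weight corrective bone transforms that reposition the hair.
# Gravity and acceleration scale vary with character state, skydiving, crouching and backbling.
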AnimBlueprint /BRCosmetics/Characters/Player/Male/Medium/Bodies/M_MED_Ensemble_Snake/Meshes/Parts/M_MED_Ensemble_Snake_FaceAcc_AnimBP
target_skeleton=&/M_MED_Ensemble_Snake_FaceAcc_Skeleton
graphs [
 AnimGraph {nodes [
   Look At - Bone: dyn_hair_side_r_AIM(Alpha=1) {
    alphaBoolBlend {blendOption=Linear}
    boneToModify {boneName=dyn_hair_side_r_AIM}
    lookAtLocation=<-20 -24.3861 13.7024>
    lookAtTarget {boneReference {boneName=spine_05}}
    lookAt_Axis {axis=<1 0>}
   }
   "Transform (Modify) Bone - Bone: dyn_hair_side_r_MSTR"(Alpha=0, Translation="0,0,0", Rotation="0,8,0", Scale="0.875,1.000,1.000") {
    rotationMode=BMM_Additive
    rotationSpace=BCS_BoneSpace
    scaleMode=BMM_Additive
    scaleSpace=BCS_BoneSpace
    translationMode=BMM_Additive
    translationSpace=BCS_BoneSpace
    alphaBoolBlend {blendOption=Linear}
    boneToModify {boneName=dyn_hair_side_r_MSTR}
    rotation {yaw=8}
    scale {x=0.875}
   }
   Output Pose(Result=Blend Poses by bool(
     BlendPose_0=Use cached pose 'rigidBodyPose'()
     BlendPose_1=Use cached pose 'noRigidBodyPose'()
     BlendTime_0=0.3
     BlendTime_1=0.3
     bActiveValue=Get bRigidBodyEnabled()
    ) {blendType=Linear}) {layerGroup=None}
   n1 = Copy Pose From Mesh(SourceMeshComponent=Get Body_Mesh())
   n2 = Copy Pose From Mesh() {bUseAttachedParent=true}
   n3 = Anim Dynamics - Bone: dyn_earring_l_01(
    ComponentPose=Local To Component(LocalPose=Layered blend per bone(BasePose=@n2, BlendPoses_0=@n1) {layerSetup [{branchFilters [{boneName=upperarm_l}, {boneName=upperarm_r}]}]})
    Alpha=1
    ExternalForce="0,0,0"
   ) {
    angularDampingOverride=0.999
    angularSpringConstant=1000
    bAngularSpring=true
    bOverrideAngularDamping=true
    bUseGravityOverride=true
    simulationSpace=RootRelative
    alphaBoolBlend {blendOption=Linear}
    boundBone {boneName=dyn_earring_l_01}
    componentAppliedLinearAccClamp=<5000 5000 5000>
    componentLinearAccScale=<0.3 0.3 0.3>
    componentLinearVelScale=<0.3 0.3 0.3>
    gravityOverride {z=-980}
    physicsBodyDefinitions [{
      boundBone {boneName=dyn_earring_l_01}
      boxExtents=<10 20 20>
      constraintSetup {angularLimitsMax=<0 5 45>, angularLimitsMin=<0 -5 -45>, angularTarget=<0 0 0>}
      localJointOffset=<3.5 0 0>
     }]
   }
   n4 = Anim Dynamics - Bone: dyn_earring_r_01(ComponentPose=@n3, Alpha=1, ExternalForce="0,0,0") {
    angularDampingOverride=0.999
    angularSpringConstant=1000
    bAngularSpring=true
    bOverrideAngularDamping=true
    bUseGravityOverride=true
    simulationSpace=RootRelative
    alphaBoolBlend {blendOption=Linear}
    boundBone {boneName=dyn_earring_r_01}
    componentAppliedLinearAccClamp=<5000 5000 5000>
    componentLinearAccScale=<0.3 0.3 0.3>
    componentLinearVelScale=<0.3 0.3 0.3>
    gravityOverride {z=-980}
    physicsBodyDefinitions [{
      boundBone {boneName=dyn_earring_r_01}
      boxExtents=<10 20 20>
      constraintSetup {angularLimitsMax=<0 5 45>, angularLimitsMin=<0 -5 -45>, angularTarget=<0 0 0>}
      localJointOffset=<3.5 0 0>
     }]
   }
   n5 = Anim Dynamics - Bone: dyn_hair_tips_03!(ComponentPose=@n4, Alpha=1, GravityScale:float=1, GravityOverride:Vector="0,0,-980", ExternalForce="0,0,0") -> (Pose) {
    angularDampingOverride=0.9999
    angularSpringConstant=1000
    bAngularSpring=true
    bGravityOverrideInSimSpace=true
    bOverrideAngularDamping=true
    simulationSpace=RootRelative
    alphaBoolBlend {blendOption=Linear}
    boundBone {boneName=dyn_hair_tips_03}
    componentAppliedLinearAccClamp=<5000 5000 5000>
    componentLinearAccScale=<0.6 0.6 0.6>
    componentLinearVelScale=<0.6 0.6 0.6>
    gravityOverride {z=-980}
    physicsBodyDefinitions [{
      sphereCollisionRadius=0
      boundBone {boneName=dyn_hair_tips_03}
      boxExtents=<12 10 10>
      constraintSetup {angularLimitsMax=<0 0 20>, angularLimitsMin=<0 -20 -20>, angularTarget=<0 0 0>}
      localJointOffset=<1 0 0>
     }]
    relativeSpaceBone {boneName=head}
   }
   n6 = "Anim Dynamics - Chain: dyn_hair_tips_01 -> dyn_hair_tips_02"!(ComponentPose=@n5, Alpha=1, GravityScale:float=1, GravityOverride:Vector="0,0,-980", ExternalForce="0,0,0") -> (Pose) {
    angularDampingOverride=0.9999
    angularSpringConstant=1000
    bAngularSpring=true
    bChain=true
    bGravityOverrideInSimSpace=true
    bOverrideAngularDamping=true
    simulationSpace=RootRelative
    alphaBoolBlend {blendOption=Linear}
    boundBone {boneName=dyn_hair_tips_01}
    chainEnd {boneName=dyn_hair_tips_02}
    componentAppliedLinearAccClamp=<3000 3000 3000>
    componentLinearAccScale=<0.6 0.6 0.6>
    componentLinearVelScale=<0.6 0.6 0.6>
    gravityOverride {z=-980}
    physicsBodyDefinitions [
     {
      sphereCollisionRadius=0
      boundBone {boneName=dyn_hair_tips_01}
      boxExtents=<10 16 20>
      constraintSetup {angularLimitsMax=<0 0 10>, angularLimitsMin=<0 -30 -3>, angularTarget=<0 0 0>}
      localJointOffset=<2 0 0>
     }
     {
      sphereCollisionRadius=0
      boundBone {boneName=dyn_hair_tips_02}
      boxExtents=<10 16 20>
      constraintSetup {angularLimitsMax=<0 5 10>, angularLimitsMin=<0 -25 -5>, angularTarget=<0 0 0>}
      localJointOffset=<1 0 0>
     }
    ]
    relativeSpaceBone {boneName=head}
   }
   n7 = Anim Dynamics - Bone: dyn_hair_ft_r_02!(ComponentPose=@n6, Alpha=1, GravityScale:float=1, GravityOverride:Vector="0,0,-980", ExternalForce="0,0,0") -> (Pose) {
    angularDampingOverride=0.9999
    angularSpringConstant=200
    bAngularSpring=true
    bGravityOverrideInSimSpace=true
    bOverrideAngularDamping=true
    simulationSpace=RootRelative
    alphaBoolBlend {blendOption=Linear}
    boundBone {boneName=dyn_hair_ft_r_02}
    componentAppliedLinearAccClamp=<5000 5000 5000>
    componentLinearAccScale=<0.6 0.6 0.6>
    componentLinearVelScale=<0.6 0.6 0.6>
    gravityOverride {z=-980}
    physicsBodyDefinitions [{
      sphereCollisionRadius=0
      boundBone {boneName=dyn_hair_ft_r_02}
      boxExtents=<10 15 15>
      constraintSetup {angularLimitsMax=<0 5 10>, angularLimitsMin=<0 -5 -10>, angularTarget=<0 0 0>}
      localJointOffset=<1 0 0>
     }]
    relativeSpaceBone {boneName=head}
   }
   n8 = Anim Dynamics - Bone: dyn_hair_ft_l_02!(ComponentPose=@n7, Alpha=1, GravityScale:float=1, GravityOverride:Vector="0,0,-980", ExternalForce="0,0,0") -> (Pose) {
    angularDampingOverride=0.9999
    angularSpringConstant=200
    bAngularSpring=true
    bGravityOverrideInSimSpace=true
    bOverrideAngularDamping=true
    simulationSpace=RootRelative
    alphaBoolBlend {blendOption=Linear}
    boundBone {boneName=dyn_hair_ft_l_02}
    componentAppliedLinearAccClamp=<5000 5000 5000>
    componentLinearAccScale=<0.6 0.6 0.6>
    componentLinearVelScale=<0.6 0.6 0.6>
    gravityOverride {z=-980}
    physicsBodyDefinitions [{
      sphereCollisionRadius=0
      boundBone {boneName=dyn_hair_ft_l_02}
      boxExtents=<10 15 15>
      constraintSetup {angularLimitsMax=<0 5 5>, angularLimitsMin=<0 -5 -5>, angularTarget=<0 0 0>}
      localJointOffset=<1 0 0>
     }]
    relativeSpaceBone {boneName=head}
   }
   Save cached pose 'BasePose'(Pose=Component To Local(ComponentPose=@n8)) {cacheName=BasePose}
   n9 = Get Head_Rot_Bk() -> (Head_Rot_Bk:double)
   n10 = "Transform (Modify) Bone - Bone: dyn_hair_ft_r_01"(
    ComponentPose=Local To Component(LocalPose=Use cached pose 'BasePose'())
    Alpha=@n9
    Translation="0,0,0"
    Rotation="0.000,15.000,0.000"
    Scale="1,1,1"
   ) {
    rotationMode=BMM_Additive
    rotationSpace=BCS_BoneSpace
    translationMode=BMM_Additive
    translationSpace=BCS_BoneSpace
    alphaBoolBlend {blendOption=Linear}
    boneToModify {boneName=dyn_hair_ft_r_01}
    rotation=<rot 0.0001 14.9999 0>
   }
   n11 = "Transform (Modify) Bone - Bone: dyn_hair_ft_l_01"(ComponentPose=@n10, Alpha=@n9, Translation="0,0,0", Rotation="0.000,-15,0.000", Scale="1,1,1") {
    rotationMode=BMM_Additive
    rotationSpace=BCS_BoneSpace
    translationMode=BMM_Additive
    translationSpace=BCS_BoneSpace
    alphaBoolBlend {blendOption=Linear}
    boneToModify {boneName=dyn_hair_ft_l_01}
    rotation {yaw=-15}
   }
   n12 = Get Head_Rot_Dn() -> (Head_Rot_Dn:double)
   n13 = "Transform (Modify) Bone - Bone: dyn_hair_ft_r_01"(ComponentPose=@n11, Alpha=@n12, Translation="0,0,0", Rotation="0.000,-10,0.000", Scale="1,1,1") {
    rotationMode=BMM_Additive
    rotationSpace=BCS_BoneSpace
    translationMode=BMM_Additive
    translationSpace=BCS_BoneSpace
    alphaBoolBlend {blendOption=Linear}
    boneToModify {boneName=dyn_hair_ft_r_01}
    rotation {yaw=-10}
   }
   n14 = "Transform (Modify) Bone - Bone: dyn_hair_ft_l_01"(ComponentPose=@n13, Alpha=@n12, Translation="0,0,0", Rotation="0.000,10,0.000", Scale="1,1,1") {
    rotationMode=BMM_Additive
    rotationSpace=BCS_BoneSpace
    translationMode=BMM_Additive
    translationSpace=BCS_BoneSpace
    alphaBoolBlend {blendOption=Linear}
    boneToModify {boneName=dyn_hair_ft_l_01}
    rotation {yaw=10}
   }
   n15 = Get Head_Rot_Bk()
   n16 = "Transform (Modify) Bone - Bone: dyn_hair_bk_MSTR"(ComponentPose=@n14, Alpha=@n15, Translation="0,0,0", Rotation="20,0,0", Scale="1,1,1") {
    rotationMode=BMM_Additive
    rotationSpace=BCS_BoneSpace
    translationMode=BMM_Additive
    translationSpace=BCS_BoneSpace
    alphaBoolBlend {blendOption=Linear}
    boneToModify {boneName=dyn_hair_bk_MSTR}
    rotation {pitch=20}
   }
   n17 = "Transform (Modify) Bone - Bone: dyn_hair_bk_r_MSTR"(ComponentPose=@n16, Alpha=@n15, Translation="0,0,1", Rotation="10,0,0", Scale="1,1,1") {
    rotationMode=BMM_Additive
    rotationSpace=BCS_BoneSpace
    translationMode=BMM_Additive
    translationSpace=BCS_BoneSpace
    alphaBoolBlend {blendOption=Linear}
    boneToModify {boneName=dyn_hair_bk_r_MSTR}
    rotation {pitch=10}
    translation {z=1}
   }
   n18 = "Transform (Modify) Bone - Bone: dyn_hair_l_MSTR"(ComponentPose=@n17, Alpha=@n15, Translation="0,0,0", Rotation="10,0,0", Scale="1,1,1") {
    rotationMode=BMM_Additive
    rotationSpace=BCS_BoneSpace
    translationMode=BMM_Additive
    translationSpace=BCS_BoneSpace
    alphaBoolBlend {blendOption=Linear}
    boneToModify {boneName=dyn_hair_l_MSTR}
    rotation {pitch=10}
   }
   n19 = "Transform (Modify) Bone - Bone: dyn_hair_side_r_MSTR"(ComponentPose=@n18, Alpha=@n15, Translation="0,0,0", Rotation="0,30,0", Scale="1,1,1") {
    rotationMode=BMM_Additive
    rotationSpace=BCS_BoneSpace
    translationMode=BMM_Additive
    translationSpace=BCS_BoneSpace
    alphaBoolBlend {blendOption=Linear}
    boneToModify {boneName=dyn_hair_side_r_MSTR}
    rotation {yaw=30}
   }
   n20 = Get Head_Rot_R() -> (Head_Rot_R:double)
   n21 = "Transform (Modify) Bone - Bone: dyn_hair_bk_MSTR"(ComponentPose=@n19, Alpha=@n20, Translation="0,0,0", Rotation="20.000,-19.998,0", Scale="1,1,1") {
    rotationMode=BMM_Additive
    rotationSpace=BCS_BoneSpace
    translationMode=BMM_Additive
    translationSpace=BCS_BoneSpace
    alphaBoolBlend {blendOption=Linear}
    boneToModify {boneName=dyn_hair_bk_MSTR}
    rotation {pitch=20, yaw=-19.998}
   }
   n22 = "Transform (Modify) Bone - Bone: dyn_hair_bk_r_MSTR"(ComponentPose=@n21, Alpha=@n20, Translation="0,0,2.5", Rotation="25,-25,0", Scale="1,1,1") {
    rotationMode=BMM_Additive
    rotationSpace=BCS_BoneSpace
    translationMode=BMM_Additive
    translationSpace=BCS_BoneSpace
    alphaBoolBlend {blendOption=Linear}
    boneToModify {boneName=dyn_hair_bk_r_MSTR}
    rotation {pitch=25, yaw=-25}
    translation {z=2.5}
   }
   n23 = "Transform (Modify) Bone - Bone: dyn_hair_side_r_MSTR"(ComponentPose=@n22, Alpha=@n20, Translation="-2,0,-1", Rotation="9.365,45.281,3.376", Scale="1,1,1") {
    rotationMode=BMM_Additive
    rotationSpace=BCS_BoneSpace
    scaleMode=BMM_Additive
    scaleSpace=BCS_BoneSpace
    translationMode=BMM_Additive
    translationSpace=BCS_BoneSpace
    alphaBoolBlend {blendOption=Linear}
    boneToModify {boneName=dyn_hair_side_r_MSTR}
    rotation=<rot 9.365 45.2809 3.376>
    translation {x=-2, z=-1}
   }
   n24 = "Transform (Modify) Bone - Bone: dyn_hair_side_r_AIM"(ComponentPose=@n23, Alpha=1, Translation="0,0,0", Rotation="0,15,0", Scale="1,1,1") {
    rotationMode=BMM_Additive
    rotationSpace=BCS_BoneSpace
    scaleMode=BMM_Additive
    scaleSpace=BCS_BoneSpace
    translationMode=BMM_Additive
    translationSpace=BCS_BoneSpace
    alphaBoolBlend {blendOption=Linear}
    boneToModify {boneName=dyn_hair_side_r_AIM}
    rotation {yaw=15}
   }
   n25 = "Transform (Modify) Bone - Bone: dyn_hair_l_MSTR"(ComponentPose=@n24, Alpha=@n20, Translation="0,0,0", Rotation="10.000,-0.014,-0.000", Scale="1,1,1") {
    rotationMode=BMM_Additive
    rotationSpace=BCS_BoneSpace
    translationMode=BMM_Additive
    translationSpace=BCS_BoneSpace
    alphaBoolBlend {blendOption=Linear}
    boneToModify {boneName=dyn_hair_l_MSTR}
    rotation=<rot 9.9999 -0.014 0>
   }
   n26 = "Transform (Modify) Bone - Bone: dyn_hair_l_AIM"(ComponentPose=@n25, Alpha=@n20, Translation="6,-2,0", Rotation="0,0,0", Scale="1,1,1") {
    rotationMode=BMM_Additive
    rotationSpace=BCS_BoneSpace
    translationMode=BMM_Additive
    translationSpace=BCS_BoneSpace
    alphaBoolBlend {blendOption=Linear}
    boneToModify {boneName=dyn_hair_l_AIM}
    translation=<6 -2>
   }
   n27 = "Transform (Modify) Bone - Bone: dyn_hair_side_r_MSTR"(ComponentPose=@n26, Alpha=Get Shoulder_Rot_Up_R() -> (Shoulder_Rot_Up_R:double), Translation="-2,6,0", Rotation="0,0,0", Scale="1,1,1") {
    rotationMode=BMM_Additive
    rotationSpace=BCS_BoneSpace
    scaleMode=BMM_Additive
    scaleSpace=BCS_BoneSpace
    translationMode=BMM_Additive
    translationSpace=BCS_BoneSpace
    alphaBoolBlend {blendOption=Linear}
    boneToModify {boneName=dyn_hair_side_r_MSTR}
    translation=<-2 6>
   }
   n28 = Look At - Bone: dyn_hair_l_AIM(ComponentPose=@n27, Alpha=1) {
    alphaBoolBlend {blendOption=Linear}
    boneToModify {boneName=dyn_hair_l_AIM}
    lookAtLocation=<-20 -23.7686 -9.7>
    lookAtTarget {boneReference {boneName=spine_04}}
    lookAt_Axis {axis=<1 0>}
   }
   n29 = Look At - Bone: dyn_hair_bk_r_AIM(ComponentPose=@n28, Alpha=1) {
    alphaBoolBlend {blendOption=Linear}
    boneToModify {boneName=dyn_hair_bk_r_AIM}
    lookAtLocation=<-20 -29.7382 7.7314>
    lookAtTarget {boneReference {boneName=spine_04}}
    lookAt_Axis {axis=<1 0>}
   }
   n30 = Look At - Bone: dyn_hair_bk_AIM(ComponentPose=@n29, Alpha=1) {
    alphaBoolBlend {blendOption=Linear}
    boneToModify {boneName=dyn_hair_bk_AIM}
    lookAtLocation=<-20 -29.2113 0>
    lookAtTarget {boneReference {boneName=spine_04}}
    lookAt_Axis {axis=<1 0>}
   }
   n31 = RigidBody!(
    ComponentPose=@n30
    Alpha
    OverrideWorldGravity:Vector=Get GravityOverride()
    ExternalForce
    ComponentLinearAccScale:Vector=Get AccScaleOverride() -> (AccScaleOverride:Vector)
   ) -> (Pose) {
    bClampLinearTranslationLimitToRefPose=true
    bOverrideWorldGravity=true
    overridePhysicsAsset=&/M_MED_Ensemble_Snake_FaceAcc_Dyn_Hair
    simulationSpace=BaseBoneSpace
    alphaBoolBlend {blendOption=Linear}
    alphaScaleBiasClamp {interpSpeedDecreasing=4, interpSpeedIncreasing=6}
    baseBoneRef {boneName=spine_05}
    componentAppliedLinearAccClamp=<1000 1000 500>
    componentLinearAccScale=<0.5 0.18 0.2>
    componentLinearVelScale=<0.2 0.2 0.2>
    overrideWorldGravity {z=-875}
    simSpaceSettings {maxAngularAcceleration=200, maxAngularVelocity=5, maxLinearAcceleration=500, maxLinearVelocity=200}
   }
   n32 = Get Head_Rot_Bk()
   n33 = "Transform (Modify) Bone - Bone: dyn_hair_bk_01"(ComponentPose=@n31, Alpha=@n32, Translation="0,0,0", Rotation="0,0,0", Scale="0.750,1.000,1.000") {
    rotationMode=BMM_Additive
    rotationSpace=BCS_BoneSpace
    scaleMode=BMM_Additive
    scaleSpace=BCS_BoneSpace
    translationMode=BMM_Additive
    translationSpace=BCS_BoneSpace
    alphaBoolBlend {blendOption=Linear}
    boneToModify {boneName=dyn_hair_bk_01}
    scale {x=0.75}
   }
   n34 = "Transform (Modify) Bone - Bone: dyn_hair_l_01"(ComponentPose=@n33, Alpha=@n32, Translation="0,0,0", Rotation="0,0,0", Scale="0.75,1,1") {
    rotationMode=BMM_Additive
    rotationSpace=BCS_BoneSpace
    scaleMode=BMM_Additive
    scaleSpace=BCS_BoneSpace
    translationMode=BMM_Additive
    translationSpace=BCS_BoneSpace
    alphaBoolBlend {blendOption=Linear}
    boneToModify {boneName=dyn_hair_l_01}
    scale {x=0.75}
   }
   n35 = "Transform (Modify) Bone - Bone: dyn_hair_bk_r_01"(ComponentPose=@n34, Alpha=@n32, Translation="0,0,0", Rotation="0,0,0", Scale="0.75,1.000,1.000") {
    rotationMode=BMM_Additive
    rotationSpace=BCS_BoneSpace
    scaleMode=BMM_Additive
    scaleSpace=BCS_BoneSpace
    translationMode=BMM_Additive
    translationSpace=BCS_BoneSpace
    alphaBoolBlend {blendOption=Linear}
    boneToModify {boneName=dyn_hair_bk_r_01}
    scale {x=0.75}
   }
   n36 = "Transform (Modify) Bone - Bone: dyn_hair_side_r_MSTR"(ComponentPose=@n35, Alpha=Get Head_Rot_R(), Translation="0,0,0", Rotation="0,0,0", Scale="0.625,0.75,1") {
    rotationMode=BMM_Additive
    rotationSpace=BCS_BoneSpace
    scaleMode=BMM_Additive
    scaleSpace=BCS_BoneSpace
    translationMode=BMM_Additive
    translationSpace=BCS_BoneSpace
    alphaBoolBlend {blendOption=Linear}
    boneToModify {boneName=dyn_hair_side_r_MSTR}
    scale=<0.625 0.75>
   }
   Save cached pose 'rigidBodyPose'(Pose=Component To Local(ComponentPose=@n36)) {cacheName=rigidBodyPose}
   n37 = "Constraint - dyn_hair_bk_r_AIM (PP)"!(ComponentPose=Local To Component(LocalPose=Use cached pose 'BasePose'()), Alpha=1, ConstraintWeights_0=0.4, ConstraintWeights_1:float=0.6) -> (Pose) {
    alphaBoolBlend {blendOption=Linear}
    boneToModify {boneName=dyn_hair_bk_r_AIM}
    constraintSetup [{targetBone {boneName=spine_05}}, {targetBone {boneName=head}}]
    constraintWeights [1,1]
   }
   n38 = "Constraint - dyn_hair_l_AIM (PP)"!(ComponentPose=@n37, Alpha=1, ConstraintWeights_0=0.4, ConstraintWeights_1:float=0.6) -> (Pose) {
    alphaBoolBlend {blendOption=Linear}
    boneToModify {boneName=dyn_hair_l_AIM}
    constraintSetup [{targetBone {boneName=spine_05}}, {targetBone {boneName=head}}]
    constraintWeights [1,1]
   }
   n39 = "Constraint - dyn_hair_bk_AIM (PP)"!(ComponentPose=@n38, Alpha=1, ConstraintWeights_0=0.4, ConstraintWeights_1:float=0.6) -> (Pose) {
    alphaBoolBlend {blendOption=Linear}
    boneToModify {boneName=dyn_hair_bk_AIM}
    constraintSetup [{targetBone {boneName=spine_05}}, {targetBone {boneName=head}}]
    constraintWeights [1,1]
   }
   n40 = Get Head_Rot_Bk()
   n41 = "Transform (Modify) Bone - Bone: dyn_hair_bk_00"(ComponentPose=@n39, Alpha=@n40, Translation="0,0,0", Rotation="20,0,0", Scale="1.125,1.000,1.000") {
    rotationMode=BMM_Additive
    rotationSpace=BCS_BoneSpace
    scaleMode=BMM_Additive
    scaleSpace=BCS_BoneSpace
    translationMode=BMM_Additive
    translationSpace=BCS_BoneSpace
    alphaBoolBlend {blendOption=Linear}
    boneToModify {boneName=dyn_hair_bk_00}
    rotation {pitch=20}
    scale {x=1.125}
   }
   n42 = "Transform (Modify) Bone - Bone: dyn_hair_bk_r_00"(ComponentPose=@n41, Alpha=@n40, Translation="0,0,0", Rotation="25,0.067,0", Scale="1,1,1") {
    rotationMode=BMM_Additive
    rotationSpace=BCS_BoneSpace
    scaleMode=BMM_Additive
    scaleSpace=BCS_BoneSpace
    translationMode=BMM_Additive
    translationSpace=BCS_BoneSpace
    alphaBoolBlend {blendOption=Linear}
    boneToModify {boneName=dyn_hair_bk_r_00}
    rotation {pitch=25, yaw=0.067}
   }
   n43 = "Transform (Modify) Bone - Bone: dyn_hair_l_00"(ComponentPose=@n42, Alpha=@n40, Translation="0,0,2", Rotation="25,0,0", Scale="1,1,1") {
    rotationMode=BMM_Additive
    rotationSpace=BCS_BoneSpace
    scaleMode=BMM_Additive
    scaleSpace=BCS_BoneSpace
    translationMode=BMM_Additive
    translationSpace=BCS_BoneSpace
    alphaBoolBlend {blendOption=Linear}
    boneToModify {boneName=dyn_hair_l_00}
    rotation {pitch=25}
    translation {z=2}
   }
   n44 = "Constraint - dyn_hair_bk_02 (PP)"!(ComponentPose=@n43, Alpha=1, ConstraintWeights_0=0.75, ConstraintWeights_1:float=0.25) -> (Pose) {
    alphaBoolBlend {blendOption=Linear}
    boneToModify {boneName=dyn_hair_bk_02}
    constraintSetup [{targetBone {boneName=spine_05}}, {targetBone {boneName=spine_04}}]
    constraintWeights [1,1]
   }
   n45 = "Constraint - dyn_hair_bk_r_02 (PP)"!(ComponentPose=@n44, Alpha=1, ConstraintWeights_0=0.75, ConstraintWeights_1:float=0.25) -> (Pose) {
    alphaBoolBlend {blendOption=Linear}
    boneToModify {boneName=dyn_hair_bk_r_02}
    constraintSetup [{targetBone {boneName=spine_05}}, {targetBone {boneName=spine_04}}]
    constraintWeights [1,1]
   }
   n46 = "Constraint - dyn_hair_l_02 (PP)"!(ComponentPose=@n45, Alpha=1, ConstraintWeights_0=0.75, ConstraintWeights_1:float=0.25) -> (Pose) {
    alphaBoolBlend {blendOption=Linear}
    boneToModify {boneName=dyn_hair_l_02}
    constraintSetup [{targetBone {boneName=spine_05}}, {targetBone {boneName=spine_04}}]
    constraintWeights [1,1]
   }
   n47 = "Transform (Modify) Bone - Bone: dyn_hair_side_r_MSTR"(ComponentPose=@n46, Alpha=Get Head_Rot_Bk(), Translation="0,2.734,0", Rotation="0.013,40.000,0.002", Scale="1,1,1") {
    rotationMode=BMM_Additive
    rotationSpace=BCS_BoneSpace
    translationMode=BMM_Additive
    translationSpace=BCS_BoneSpace
    alphaBoolBlend {blendOption=Linear}
    boneToModify {boneName=dyn_hair_side_r_MSTR}
    rotation=<rot 0.013 39.9999 0.002>
    translation {y=2.734}
   }
   n48 = "Constraint - dyn_hair_ft_l_03 (PP)"!(ComponentPose=@n47, Alpha=1, ConstraintWeights_0=0.4, ConstraintWeights_1:float=0.6) -> (Pose) {
    alphaBoolBlend {blendOption=Linear}
    boneToModify {boneName=dyn_hair_ft_l_03}
    constraintSetup [{targetBone {boneName=spine_05}}, {targetBone {boneName=head}}]
    constraintWeights [1,1]
   }
   n49 = "Constraint - dyn_hair_ft_r_03 (PP)"!(ComponentPose=@n48, Alpha=1, ConstraintWeights_0=0.4, ConstraintWeights_1:float=0.6) -> (Pose) {
    alphaBoolBlend {blendOption=Linear}
    boneToModify {boneName=dyn_hair_ft_r_03}
    constraintSetup [{targetBone {boneName=spine_05}}, {targetBone {boneName=head}}]
    constraintWeights [1,1]
   }
   n50 = Get Head_Rot_Dn()
   n51 = "Transform (Modify) Bone - Bone: dyn_hair_ft_r_02"(ComponentPose=@n49, Alpha=@n50, Translation="0,-2,0", Rotation="0,-20.000,0", Scale="1,1,1") {
    rotationMode=BMM_Additive
    rotationSpace=BCS_BoneSpace
    translationMode=BMM_Additive
    translationSpace=BCS_BoneSpace
    alphaBoolBlend {blendOption=Linear}
    boneToModify {boneName=dyn_hair_ft_r_02}
    rotation {yaw=-20}
    translation {y=-2}
   }
   n52 = "Transform (Modify) Bone - Bone: dyn_hair_ft_l_02"(ComponentPose=@n51, Alpha=@n50, Translation="0,2,0", Rotation="0,20.000,0", Scale="1,1,1") {
    rotationMode=BMM_Additive
    rotationSpace=BCS_BoneSpace
    translationMode=BMM_Additive
    translationSpace=BCS_BoneSpace
    alphaBoolBlend {blendOption=Linear}
    boneToModify {boneName=dyn_hair_ft_l_02}
    rotation {yaw=20}
    translation {y=2}
   }
   Save cached pose 'noRigidBodyPose'(Pose=Component To Local(ComponentPose=@n52)) {cacheName=noRigidBodyPose}
   Get Shoulder_Rot_Up_R()
  ]}
 EventGraph {nodes [
   Event BlueprintUpdateAnimation()
   > n1 = Sequence()
   n2 = GetOwnerSkeletalMeshComponent()
   n3 = BreakRotator(InRot=BreakTransform(InTransform=GetDeltaTransformFromRefPose(self=@n2, BoneName=head, BaseName=spine_05)).Rotation)
   > Set Head_Rot_R(Head_Rot_R:double=MapRangeClamped(Value=@n3.Pitch, InRangeA=15, InRangeB=52)) -> (Output_Get:double)
   > Set Head_Rot_Bk(Head_Rot_Bk:double=MapRangeClamped(Value=@n3.Yaw, InRangeA=-10, InRangeB=-55)) -> (Output_Get:double)
   > Set Head_Rot_Dn(Head_Rot_Dn:double=MapRangeClamped(Value=@n3.Yaw, InRangeA=0, InRangeB=30)) -> (Output_Get:double)
   n4 = BreakTransform(InTransform=GetDeltaTransformFromRefPose(self=GetOwnerSkeletalMeshComponent(), BoneName=clavicle_r, BaseName=spine_05))
   > Set Shoulder_Rot_Up_R(Shoulder_Rot_Up_R:double=MapRangeClamped(Value=BreakRotator(InRot=@n4.Rotation).Pitch, InRangeA=0, InRangeB=35)) -> (Output_Get:double)
   n5 = Gravity Override State Function DEPRECATED(
    execute=@n1.then_1
    CharacterPartAnimInstance=Self-Reference()
    GameTimeInSeconds=GetGameTimeInSeconds()
    GravityOverrideParams=Get Gravity Override Params()
   )
   > n6 = Set GravityOverride(GravityOverride=@n5)
   > n7 = Sequence()
   > n8 = Branch(Condition=OR Boolean(A=Get bIsSkyDiving(), B=Get bIsCrouching()))
   > Set AccScaleOverride(AccScaleOverride:Vector="0.1,0.1,0.05") -> (Output_Get:Vector)
   Set AccScaleOverride(execute=@n8.else, AccScaleOverride="0.3,0.25,0.2")
   n9 = Branch(execute=@n7.then_1, Condition=Get isWearBackbling() -> (isWearBackbling:bool))
   > Set AccScaleOverride(AccScaleOverride="0.1,0.1,0.1")
   > Set GravityOverride(GravityOverride="0, 0,-700")
   Set AccScaleOverride(execute=@n9.else, AccScaleOverride="0.3,0.25,0.2")
   > Set GravityOverride(GravityOverride=@n6)
   Event OnCharacterCustomizationLoaded()
   > n10 = Sequence()
   > Delay(Duration=0.1)
   > ResetDynamics(InTeleportType=TeleportPhysics)
   > ResetDynamics(InTeleportType=ResetPhysics)
   Do Once(execute=@n10.then_1)
   > Set isWearBackbling(isWearBackbling:bool=CheckForBackbling()) -> (Output_Get:bool)
   Event BlueprintInitializeAnimation()
   > Set Body_Mesh(Body_Mesh=GetOwnerSkeletalMeshComponent())
  ]}
]
variables [
 AccScaleOverride {type=Vector}
 Body_Mesh {type=SkeletalMeshComponent}
 Gravity Override Params {
  type=GravityOverrideParamsStruct
  default {
   dBNO {gravityOverride=<-700 200 0>}
   dBNO_Carried {gravityOverride=<-700 300 -300>}
   emoteOrMelee {gravityOverride=<0 0 -700>}
   front_End {gravityOverride=<0 0 -700>}
   onGround_Falling {gravityOverride=<0 0 -700>}
   onGround_Moving {gravityOverride=<0 0 -700>}
   onGround_Standing {gravityOverride=<0 0 -700>}
   skydive_DivingDown {gravityOverride=<-1500 600 0>, windAmplitude=<50 50 100>}
   skydive_FloatingDown {gravityOverride=<0 -100 -700>, windAmplitude=<50 -100 50>, windOffset=<0 0 50>}
   skydive_GoingUp {gravityOverride=<0 -100 -300>, windAmplitude=<50 -200 100>}
   skydive_Parachuting {gravityOverride=<0 -500 -700>, windAmplitude=<100 -150 200>}
   swimming {gravityOverride=<-500 200 0>}
   ziplining {gravityOverride=<0 0 -700>}
  }
 }
 GravityOverride {type=Vector}
 Head_Rot_Bk {type=number}
 Head_Rot_Dn {type=number}
 Head_Rot_R {type=number}
 isWearBackbling {type=boolean}
 Shoulder_Rot_Up_R {type=number}
]
